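# Material function that applies a biome color tint to tintable layers, named for use on rock materials.
# It lerps the input Material Attributes' BaseColor toward that of the biome color attributes, apparently by the Tintable Mask, leaving other attributes untouched.
# A static switch, fed by the "Use Biome Tint on Tintable Layers" option, outputs either the tinted or the unchanged attributes.
# Inputs are routed through named reroutes, with an unconnected "Masks" input acting as a pin-list divider.
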MaterialFunction /BRCoreMaterials/Materials/MaterialFunctions/MF_Rocks_Biome_Tint
graph {nodes [
  FunctionInput() {inputName="\\ --- Masks --- /", inputType=FunctionInput_Scalar, sortPriority=100}
  n1 = NamedRerouteUsage()
  n2 = GetMaterialAttributes(None=@n1) {attributeGetTypes [#1]}
  n3 = LinearInterpolate(
   A=@n2.BaseColor
   B=GetMaterialAttributes(None=NamedRerouteUsage()) {attributeGetTypes [#1]}.BaseColor
   Alpha=NamedRerouteUsage()
  )
  FunctionOutput(None=StaticSwitch(
    True=SetMaterialAttributes!(MaterialAttributes=@n2.MaterialAttributes, BaseColor=@n3) {attributeSetTypes [#1]}
    False=@n1
    Value=NamedRerouteUsage()
   )) {outputName=Blended MaterialAttributes, sortPriority=-100}
  n4 = FunctionInput(Preview=StaticBool()) {bUsePreviewValueAsDefault=true, inputName=Use Biome Tint on Tintable Layers, inputType=FunctionInput_StaticBool, sortPriority=-110}
  NamedRerouteDeclaration(None=@n4) {name="Use Biome Tint on Tintable Layers(SB)", nodeColor=<rgba 0 0 1 1>}
  n5 = SetMaterialAttributes!(MaterialAttributes)
  n6 = FunctionInput(Preview=@n5) {bUsePreviewValueAsDefault=true, inputName=Biome Color MaterialAttributes, inputType=FunctionInput_MaterialAttributes, sortPriority=-20}
  NamedRerouteDeclaration(None=@n6) {name=Biome Color MaterialAttributes, nodeColor=<rgba 0.4908 0 1 1>}
  NamedRerouteDeclaration(None=FunctionInput(Preview=@n5) {inputName=Input MaterialAttributes, inputType=FunctionInput_MaterialAttributes, sortPriority=-100}) {name=_Input MaterialAttributes, nodeColor=<rgba 0.1016 0 1 1>}
  NamedRerouteDeclaration(None=FunctionInput() {bUsePreviewValueAsDefault=true, inputName=Tintable Mask, inputType=FunctionInput_Scalar, sortPriority=101}) {name=_Tintable Mask, nodeColor=<rgba 1 1 1 1>}
 ]}
