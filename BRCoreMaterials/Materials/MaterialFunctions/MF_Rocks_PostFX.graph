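# Material function that applies post effects to rock MaterialAttributes: RVT blending, wetness, realtime displacement and destructible behavior.
# The first three run as rock-specific sub-functions, fed by static-bool toggles, masks and RVT inputs through named reroutes.
# A static switch optionally adds destructible damage and WPO bob-and-rotate before MF_CreativeMoveTool produces the output.
# A Curie fallback switch and a normal-transform branch are present but do not feed the output.
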MaterialFunction /BRCoreMaterials/Materials/MaterialFunctions/MF_Rocks_PostFX
graph {nodes [
  Constant() {r=1}
  FunctionInput() {inputName="\\ --- Masks --- /", inputType=FunctionInput_Scalar, sortPriority=50}
  FunctionInput() {inputName="\\ --- Dependencies --- /", inputType=FunctionInput_Scalar, sortPriority=800}
  FunctionInput() {inputName="\\ --- RVT --- /", inputType=FunctionInput_Scalar, sortPriority=900}
  n1 = MaterialFunctionCall!(In=MaterialFunctionCall(Material_In=NamedRerouteUsage()) {materialFunction=&MF_FN_DestructibleDamage}.Material_Out) {materialFunction=/Game/Global/MaterialFunctions/MF_WPOFloatBobAndRotate}
  n2 = NamedRerouteUsage()
  n3 = NamedRerouteUsage()
  FunctionOutput(None=MaterialFunctionCall(MA In=StaticSwitch(True=@n1.Result, False=@n2, Value=@n3)) {materialFunction=&MF_CreativeMoveTool}.MA Out) {outputName=Blended MaterialAttributes}
  NamedRerouteDeclaration(None=FunctionInput(Preview=StaticBool()) {bUsePreviewValueAsDefault=true, inputName=Use RVT, inputType=FunctionInput_StaticBool, sortPriority=-115}) {name=Use RVT, nodeColor=<rgba 0 0 1 1>}
  NamedRerouteDeclaration(None=FunctionInput() {inputName=Z Gradient, inputType=FunctionInput_Scalar, sortPriority=814}) {name=_Z Gradient, nodeColor=<rgba 1 0 0.0525 1>}
  NamedRerouteDeclaration(None=FunctionInput() {inputName=RVT Base Color, sortPriority=905}) {name=_RVT Base Color, nodeColor=<rgba 0.3749 0.6555 1 0.5>}
  NamedRerouteDeclaration(None=MaterialFunctionCall!(
    Use RVT=NamedRerouteUsage()
    Input MaterialAttributes=NamedRerouteUsage()
    "\\ --- RVT --- /"
    RVT Base Color=NamedRerouteUsage()
    RVT Specular=NamedRerouteUsage()
    RVT Roughness=NamedRerouteUsage()
    RVT Normal=NamedRerouteUsage()
    RVT Height=NamedRerouteUsage()
   ) {materialFunction=&/MF_Rocks_RVT_Blending}.Blended MaterialAttributes) {name=RVT Blended MaterialAtributes, nodeColor=<rgba 0.0547 0 1 1>}
  NamedRerouteDeclaration(None=FunctionInput() {inputName=RVT Normal, sortPriority=909}) {name=_RVT Normals, nodeColor=<rgba 0.3749 0.6555 1 0.5>}
  n4 = NamedRerouteUsage()
  NamedRerouteDeclaration(None=MaterialFunctionCall!(Use Wetness=NamedRerouteUsage(), Input MaterialAttributes=@n4, "\\ --- Masks --- /", Wetness Mask=NamedRerouteUsage()) {materialFunction=&/MF_Rocks_Wetness}.Blended MaterialAttributes) {name=Wetness Blended MaterialAttributes, nodeColor=<rgba 0.2656 1 0 1>}
  NamedRerouteDeclaration(None=MaterialFunctionCall!(Use Realtime Displacement=NamedRerouteUsage(), Input MaterialAttributes=NamedRerouteUsage()) {materialFunction=&/MF_Rocks_Realtime_Displacement}.Blended MaterialAttributes) {name=RealtimeDisplacement Blended MaterialAttributes, nodeColor=<rgba 1 0 0.5859 1>}
  NamedRerouteDeclaration(None=FunctionInput(Preview=StaticBool()) {bUsePreviewValueAsDefault=true, inputName=Is Destructible, inputType=FunctionInput_StaticBool, sortPriority=-110}) {name=Is Destructible, nodeColor=<rgba 0 0 1 1>}
  NamedRerouteDeclaration(None=FunctionInput() {inputName=Wetness Mask, inputType=FunctionInput_Scalar, sortPriority=55}) {name=Wetness Mask, nodeColor=<rgba 0.1016 0 1 1>}
  NamedRerouteDeclaration(None=FunctionInput() {inputName=Baked AO, inputType=FunctionInput_Scalar, sortPriority=811}) {name=_Baked AO, nodeColor=<rgba 1 0 0.0525 1>}
  NamedRerouteDeclaration(None=FunctionInput() {inputName=WS Noise Masks, inputType=FunctionInput_Scalar, sortPriority=815}) {name=_WS Noise Masks, nodeColor=<rgba 1 0 0.0525 1>}
  NamedRerouteDeclaration(None=FunctionInput() {inputName=RVT Specular, inputType=FunctionInput_Scalar, sortPriority=906}) {name=_RVT Specular, nodeColor=<rgba 0.3749 0.6555 1 0.5>}
  NamedRerouteDeclaration(None=FunctionInput() {inputName=RVT Roughness, inputType=FunctionInput_Scalar, sortPriority=907}) {name=_RVT Roughness, nodeColor=<rgba 0.3749 0.6555 1 0.5>}
  NamedRerouteDeclaration(None=FunctionInput() {inputName=DFAO, inputType=FunctionInput_Scalar, sortPriority=813}) {name=_DFAO, nodeColor=<rgba 1 0 0.0525 1>}
  NamedRerouteDeclaration(None=FunctionInput() {inputName=Baked Curvature, inputType=FunctionInput_Scalar, sortPriority=810}) {name=_Baked Curvature, nodeColor=<rgba 1 0 0.0525 1>}
  NamedRerouteDeclaration(None=FunctionInput(Preview=SetMaterialAttributes!(MaterialAttributes)) {inputName=Input MaterialAttributes, inputType=FunctionInput_MaterialAttributes, sortPriority=-100}) {name=_Input MaterialAttributes, nodeColor=<rgba 0.1016 0 1 1>}
  NamedRerouteDeclaration(None=FunctionInput() {inputName=Painted Mask, inputType=FunctionInput_Scalar, sortPriority=812}) {name=_Painted Mask, nodeColor=<rgba 1 0 0.0525 1>}
  NamedRerouteDeclaration(None=FunctionInput(Preview=StaticBool()) {bUsePreviewValueAsDefault=true, inputName=Use Wetness, inputType=FunctionInput_StaticBool, sortPriority=-120}) {name=Use Wetness, nodeColor=<rgba 0 0 1 1>}
  n5 = FunctionInput(Preview=StaticBool()) {bUsePreviewValueAsDefault=true, inputName=Use Realtime Displacement, inputType=FunctionInput_StaticBool, sortPriority=-110}
  NamedRerouteDeclaration(None=@n5) {name=Use Realtime Displacement, nodeColor=<rgba 0 0 1 1>}
  NamedRerouteDeclaration(None=FunctionInput() {inputName=RVT Height, inputType=FunctionInput_Scalar, sortPriority=920}) {name=_RVT Height, nodeColor=<rgba 0.3749 0.6555 1 0.5>}
  n6 = GetMaterialAttributes(None=@n4) {attributeGetTypes [#1]}
  SetMaterialAttributes!(MaterialAttributes=@n6.MaterialAttributes, Normal=Transform(None=@n6.Normal)) {attributeSetTypes [#1]}
  StaticSwitchParameter(
   True=MaterialFunctionCall!(In=@n1.Result, UseSphericalMappingOverride) {materialFunction=/Game/Packages/Fortress_SharedMaterials/Base_Material_Functions/MF_CurieFallbackEffect}.Result
   False=@n1.Result
  ) {defaultValue=true, parameterName=Use Curie?}
  StaticSwitch()
 ]}
