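# Material function that builds the pre-FX material attributes for rocks and cliffs from switchable Base, Detail, Top and Extra layers, plus curvature and AO coloring.
# It computes shared layer inputs: local-bounds Z gradient, wavy-distorted world position, triplanar world-space noise masks and DFAO.
# DFAO uses depth-faded distance-field AO at Epic quality, with a baked-texture fallback elsewhere and on mobile.
# It outputs attributes with a transformed normal, plus those masks for downstream use.
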MaterialFunction /BRCoreMaterials/Materials/MaterialFunctions/MF_Rocks_PreFX
graph {nodes [
  FunctionOutput(None=NamedRerouteUsage()) {outputName=Baked Curvature, sortPriority=15}
  n1 = GetMaterialAttributes(None=NamedRerouteUsage()) {attributeGetTypes [#1]}
  FunctionOutput(None=SetMaterialAttributes!(MaterialAttributes=@n1.MaterialAttributes, Normal=Transform(None=@n1.Normal)) {attributeSetTypes [#1]}) {outputName=Rocks PreFX MaterialAttributes, sortPriority=-100}
  FunctionOutput(None=NamedRerouteUsage()) {outputName=L3 - Blend Mask, sortPriority=55}
  FunctionOutput(None=NamedRerouteUsage()) {outputName=L4 - Blend Mask, sortPriority=60}
  FunctionOutput(None=NamedRerouteUsage()) {outputName=Wavy Distortion, sortPriority=30}
  FunctionOutput() {outputName="\\ --- Dependencies Output --- /", sortPriority=1}
  FunctionOutput(None=NamedRerouteUsage()) {outputName=DFAO, sortPriority=25}
  FunctionOutput(None=NamedRerouteUsage()) {outputName=Z Gradient, sortPriority=20}
  FunctionOutput(None=NamedRerouteUsage()) {outputName=WS Noise Masks, sortPriority=30}
  FunctionOutput(None=NamedRerouteUsage()) {outputName=PreFX Tintable Mask, sortPriority=-50}
  FunctionOutput(None=NamedRerouteUsage()) {outputName=Painted Mask, sortPriority=17}
  FunctionOutput(None=NamedRerouteUsage()) {outputName=Baked AO, sortPriority=16}
  FunctionOutput(None=NamedRerouteUsage()) {outputName=L2 - Blend Mask, sortPriority=50}
  n2 = MaterialFunctionCall!(Layer Is Used=NamedRerouteUsage(), "\\ --- Dependencies --- /", Wavy Distortion World Position=NamedRerouteUsage()) {materialFunction=&/MF_Rocks_L1_Base}
  NamedRerouteDeclaration(None=@n2.Baked AO) {name=_Baked AO, nodeColor=<rgba 0 0 0.5469 1>}
  NamedRerouteDeclaration(None=@n2.Baked Curvature) {name=_Baked Curvature, nodeColor=<rgba 0 0 0.5469 1>}
  n3 = MaterialFunctionCall!(
   Layer Is Used=NamedRerouteUsage()
   Input MaterialAttributes=NamedRerouteUsage()
   Input Tintable Mask=NamedRerouteUsage()
   "\\ --- Dependencies --- /"
   Baked Curvature=NamedRerouteUsage()
   Baked AO=NamedRerouteUsage()
   Painted Mask=NamedRerouteUsage()
   DFAO=NamedRerouteUsage()
   Z Gradient=NamedRerouteUsage()
   WS Noise Masks=NamedRerouteUsage()
   Wavy Distortion World Position=NamedRerouteUsage()
  ) {materialFunction=&/MF_Rocks_L3_Top}
  NamedRerouteDeclaration(None=@n3.Blended Tintable Mask) {name=L3 - Blended Tintable Mask, nodeColor=<rgba 0.1042 0.2143 1 1>}
  NamedRerouteDeclaration(None=@n3.Blend Mask) {name=L3 - Blend Mask, nodeColor=<rgba 0.1042 0.2143 1 1>}
  n4 = MaterialFunctionCall!(
   Layer Is Used=NamedRerouteUsage()
   Input MaterialAttributes=NamedRerouteUsage()
   Input Tintable Mask=NamedRerouteUsage()
   "\\ --- Dependencies --- /"
   Baked Curvature=NamedRerouteUsage()
   Baked AO=NamedRerouteUsage()
   Painted Mask=NamedRerouteUsage()
   DFAO=NamedRerouteUsage()
   Z Gradient=NamedRerouteUsage()
   WS Noise Masks=NamedRerouteUsage()
   Wavy Distortion World Position=NamedRerouteUsage()
  ) {materialFunction=&/MF_Rocks_L2_Detail}
  NamedRerouteDeclaration(None=@n4.Layer MaterialAttributes) {name=L2 - MaterialAttributes, nodeColor=<rgba 0 1 0.2598 1>}
  NamedRerouteDeclaration(None=MaterialFunctionCall!(
    Use Baked Curvature=NamedRerouteUsage()
    Use Baked AO=NamedRerouteUsage()
    Input MaterialAttributes=NamedRerouteUsage()
    "\\ --- Masks --- /"
    Tintable Mask=NamedRerouteUsage()
    "\\ --- Dependencies --- /"
    Baked Curvature=NamedRerouteUsage()
    Baked AO=NamedRerouteUsage()
    DFAO=NamedRerouteUsage()
   ) {materialFunction=&/MF_Rocks_Curvature_AO}.Blended MaterialAttributes) {name=Curvature Dirt - Blended MaterialAttributes, nodeColor=<rgba 1 0.2344 0 1>}
  NamedRerouteDeclaration(None=FunctionInput(Preview=StaticBool()) {bUsePreviewValueAsDefault=true, inputName=Use Curvature Coloring, inputType=FunctionInput_StaticBool, sortPriority=20}) {name=_Use Curvature, nodeColor=<rgba 0 0 0 1>}
  NamedRerouteDeclaration(None=FunctionInput(Preview=StaticBool()) {bUsePreviewValueAsDefault=true, inputName=Use AO Coloring, inputType=FunctionInput_StaticBool, sortPriority=25}) {name=_Use Dirt, nodeColor=<rgba 0 0 0 1>}
  NamedRerouteDeclaration(None=FunctionInput(Preview=StaticBool()) {bUsePreviewValueAsDefault=true, inputName=Use L4 | Extra, inputType=FunctionInput_StaticBool, sortPriority=16}) {name=_Use L4 - Extra, nodeColor=<rgba 0 0 0 1>}
  NamedRerouteDeclaration(None=@n3.Layer MaterialAttributes) {name=L3 - MaterialAttributes, nodeColor=<rgba 0.1042 0.2143 1 1>}
  NamedRerouteDeclaration(None=FunctionInput(Preview=StaticBool()) {bUsePreviewValueAsDefault=true, inputName=Use Realtime Displacement, inputType=FunctionInput_StaticBool}) {name=_Use Realtime Displacement, nodeColor=<rgba 0 0 0 1>}
  n5 = MaterialFunctionCall!(
   Layer Is Used=NamedRerouteUsage()
   Input MaterialAttributes=NamedRerouteUsage()
   Input Tintable Mask=NamedRerouteUsage()
   "\\ --- Dependencies --- /"
   Baked Curvature=NamedRerouteUsage()
   Baked AO=NamedRerouteUsage()
   Painted Mask=NamedRerouteUsage()
   DFAO=NamedRerouteUsage()
   Z Gradient=NamedRerouteUsage()
   WS Noise Masks=NamedRerouteUsage()
   Wavy Distortion World Position=NamedRerouteUsage()
  ) {materialFunction=&/MF_Rocks_L4_Extra}
  NamedRerouteDeclaration(None=@n5.Blended MaterialAttributes) {name=L4 - Blended MaterialAttributes, nodeColor=<rgba 0.6302 0.6757 1 1>}
  n6 = MaterialFunctionCall() {materialFunction=&ObjectLocalBounds}
  n7 = Subtract(A=LocalPosition().XYZ, B=@n6.Local Bounds Minimum)
  NamedRerouteDeclaration(None=ComponentMask(None=Divide(A=@n7, B=@n6.Local Bounds Size)) {b=true}) {name=_Z Gradient, nodeColor=<rgba 0 0 0.5469 1>}
  NamedRerouteDeclaration(None=@n5.Blended Tintable Mask) {name=L4 - Blended Tintable Mask, nodeColor=<rgba 0.6302 0.6757 1 1>}
  NamedRerouteDeclaration(None=@n5.Blend Mask) {name=L4 - Blend Mask, nodeColor=<rgba 0.6302 0.6757 1 1>}
  NamedRerouteDeclaration(None=@n5.Layer MaterialAttributes) {name=L4 - MaterialAttributes, nodeColor=<rgba 0.6302 0.6757 1 1>}
  n8 = ScalarParameter() {defaultValue=2500, group="--       Rock/Cliff Shared settings", parameterName=* Wavy Distortion - Intensity, sortPriority=200}
  n9 = VectorParameter() {group="--       Rock/Cliff Shared settings", parameterName=* Wavy Distortion - Offset, sortPriority=205, channelNames {a="", b=Z, g=Y, r=X}}
  n10 = ScalarParameter() {defaultValue=10000, group="--       Rock/Cliff Shared settings", parameterName=* Wavy Distortion - Frequency, sortPriority=210}
  n11 = MaterialFunctionCall!(WorldPosition, Wavy Distortion - Intensity=@n8, Wavy Distortion - Position Offset=@n9, Wavy Distortion - Frequency=@n10) {materialFunction=&/MF_WavyDistortion}
  NamedRerouteDeclaration(None=@n11.Wavy Distortion World Position) {name=_Wavy Distortion World Position, nodeColor=<rgba 0 0 0.5469 1>}
  NamedRerouteDeclaration(None=@n11.Wavy Distortion) {name=_Wavy Distortion, nodeColor=<rgba 0 0 0.5469 1>}
  NamedRerouteDeclaration(None=@n2.Painted Mask) {name=_Painted Mask, nodeColor=<rgba 0 0 0.5469 1>}
  n12 = ScalarParameter() {
   defaultValue=1
   desc="Set to \"0\" if no AO texture is desired"
   group="--       Rock/Cliff Shared settings"
   parameterName=* DFAO - UseBakedAOTexture
   sortPriority=50
  }
  n13 = LinearInterpolate(A=Constant() {r=1}, B=NamedRerouteUsage(), Alpha=Saturate(None=@n12))
  n14 = ScalarParameter() {defaultValue=55, desc=55, group="--       Rock/Cliff Shared settings", parameterName=* DFAO - RVT Distance, sortPriority=55}
  n15 = ScalarParameter() {defaultValue=150, desc=150, group="--       Rock/Cliff Shared settings", parameterName=* DFAO - RVT Radius, sortPriority=56}
  n16 = OneMinus(None=DistanceFieldApproxAO!(World Position, Normal, BaseDistance=@n14, Radius=@n15) {numSteps=3})
  n17 = PixelDepth()
  n18 = Constant() {r=5000}
  n19 = Saturate(None=Divide(A=Subtract(A=@n17.R, B=@n18), B=Constant() {r=10000}))
  n20 = OneMinus(None=Saturate(None=Subtract(A=@n16, B=@n19)))
  n21 = ScalarParameter() {defaultValue=1, group="--       Rock/Cliff Shared settings", parameterName=* DFAO - Power, sortPriority=52}
  n22 = Saturate(None=Multiply(
    A=Power(Base=@n20, Exp=@n21)
    B=ScalarParameter() {defaultValue=1, group="--       Rock/Cliff Shared settings", parameterName=* DFAO - Multiply, sortPriority=51}
   ))
  NamedRerouteDeclaration(None=ShadingPathSwitch(Default=QualitySwitch(Default=@n13, Epic=@n22), Mobile=@n13)) {name=_DFAO, nodeColor=<rgba 0 0 0.5469 1>}
  NamedRerouteDeclaration(None=@n4.Blended MaterialAttributes) {name=L2 - Blended MaterialAttributes, nodeColor=<rgba 0 1 0.2598 1>}
  NamedRerouteDeclaration(None=@n4.Blended Tintable Mask) {name=L2 - Blended Tintable Mask, nodeColor=<rgba 0 1 0.2598 1>}
  NamedRerouteDeclaration(None=@n2.Layer MaterialAttributes) {name=L1 - MaterialAttributes, nodeColor=<rgba 1 0.0389 0 1>}
  NamedRerouteDeclaration(None=@n2.Layer Tintable Mask) {name=L1 - Tintable Mask, nodeColor=<rgba 1 0.0389 0 1>}
  NamedRerouteDeclaration(None=FunctionInput(Preview=StaticBool() {value=true}) {bUsePreviewValueAsDefault=true, inputName=Use L1 | Base, inputType=FunctionInput_StaticBool, sortPriority=5}) {name=_Use L1 - Base, nodeColor=<rgba 0 0 0 1>}
  NamedRerouteDeclaration(None=FunctionInput(Preview=StaticBool()) {bUsePreviewValueAsDefault=true, inputName=Use L2 | Detail, inputType=FunctionInput_StaticBool, sortPriority=10}) {name=_Use L2 - Detail, nodeColor=<rgba 0 0 0 1>}
  NamedRerouteDeclaration(None=FunctionInput(Preview=StaticBool()) {bUsePreviewValueAsDefault=true, inputName=Use L3 | Top, inputType=FunctionInput_StaticBool, sortPriority=15}) {name=_Use L3 - Top, nodeColor=<rgba 0 0 0 1>}
  NamedRerouteDeclaration(None=@n4.Blend Mask) {name=L2 - Blend Mask, nodeColor=<rgba 0 1 0.2598 1>}
  NamedRerouteDeclaration(None=@n3.Blended MaterialAttributes) {name=L3 - Blended MaterialAttributes, nodeColor=<rgba 0.1042 0.2143 1 1>}
  n23 = TextureObjectParameter() {
   group="--       Rock/Cliff Shared settings"
   parameterName=* WS Noise Masks - Masks Texture
   samplerType=SAMPLERTYPE_Masks
   sortPriority=105
   texture=/Game/Global/Textures/Default/Blanks/T_MidgreyMask
  }
  n24 = VectorParameter() {
   group="--       Rock/Cliff Shared settings"
   parameterName=* WS Noise Masks - XYZ Texture Scale
   sortPriority=100
   channelNames {a="", b=Z - Scale, g=Y - Scale, r=X - Scale}
   defaultValue=<rgba 700 700 700 1>
  }
  n25 = Constant() {r=75}
  n26 = Constant() {r=30}
  n27 = MaterialFunctionCall!(
   Use Realtime Displacement=StaticBoolParameter() {group="--       Rock/Cliff Shared settings", parameterName=* WS Noise Masks - Use Triplanar Override}
   Use Expensive Triplanar=StaticBoolParameter() {group="--       Rock/Cliff Shared settings", parameterName=* WS Noise Masks - Use Costly Triplanar Override}
   Diffuse Texture Input=@n23
   Texture Scale=@n24.RGB
   Absolute World Position
   "VertexNormal (WS)"
   Triplanar - Fade=MaterialFunctionCall!(Low=@n25, High=@n26, Epic=@n26, --------=@n26, Mobile=@n25, Medium=@n26) {materialFunction=/FortniteBaseMaterials/Functions/MF_QualitySwitch_Scalar}.Result
   Derivatives Scale
   Triplanar - Flat - Seam Sharpness
  ) {materialFunction=&/MF_Rocks_Triplanar_Displacement}
  NamedRerouteDeclaration(None=@n27.RGB) {name=_WS Noise Masks, nodeColor=<rgba 0 0 0.5469 1>}
  NamedRerouteUsage()
  TextureSampleParameter2D(UVs=@n27.Triplanar Dithered - UVs) {
   group="--       Rock/Cliff Shared settings"
   parameterName=* WS Noise Masks - Masks Texture
   samplerType=SAMPLERTYPE_Masks
   sortPriority=105
   texture=&T_MidgreyMask
  }
 ]}
