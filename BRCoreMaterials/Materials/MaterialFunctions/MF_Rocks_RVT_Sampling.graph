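# Material function that samples the Apollo landscape Runtime Virtual Textures for rocks, outputting base color, specular, roughness, normals and height.
# An optional inside projection shifts the sampling position along the horizontal vertex normal, weighted toward steep faces by a power falloff and blend.
# A static switch picks between feedback sampling and mip 0 sampling without feedback.
# Height compares world Z against the height RVT, and the height output passes through an RVT on/off switch.
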MaterialFunction /BRCoreMaterials/Materials/MaterialFunctions/MF_Rocks_RVT_Sampling
graph {nodes [
  Constant3Vector() {constant=<rgba 1 0 0 0>}
  n1 = NamedRerouteUsage()
  n2 = Constant() {r=1}
  FunctionOutput(None=MaterialFunctionCall!(NoRVT=Constant() {r=1}, RVT=If(A=@n1, "A > B"=@n1, "A < B"=@n2)) {materialFunction=/Game/Packages/Fortress_SharedMaterials/Base_Material_Functions/FN_RVT_Switch_Scalar}.Result) {outputName=RVT Height, sortPriority=2}
  n3 = WorldPosition() {worldPositionShaderOffset=WPT_ExcludeAllShaderOffsets}
  n4 = AppendVector(A=ComponentMask(None=VertexNormalWS()) {g=true, r=true}, B=Constant())
  n5 = Constant3Vector() {constant=<rgba 0 0 1 0>}
  n6 = DotProduct(A=VertexNormalWS(), B=@n5)
  n7 = OneMinus(None=Power(
    Base=Saturate(None=@n6)
    Exp=ScalarParameter() {defaultValue=4, group=--- F5 | RVT, parameterName=* RVT - Inside Projection - Power, sortPriority=-9}
   ))
  n8 = Multiply(
   A=Multiply(
    A=MaterialFunctionCall(Vector=@n4, Default=@n5) {materialFunction=&SafeNormalize}.Result
    B=NamedRerouteUsage()
   )
   B=@n7
  )
  n9 = ScalarParameter() {
   defaultValue=1
   desc="Blends how much influence the Inside Projection has on the RVT sampling, a good value around steep areas is ~0.2\r\n0 = No effect\r\n1 = Full effect"
   group=--- F5 | RVT
   parameterName=* RVT - Inside Projection - Blend
   sortPriority=-8
  }
  n10 = StaticSwitchParameter(True=LinearInterpolate(A=@n3.XYZ, B=Subtract(A=@n3.XYZ, B=@n8), Alpha=@n9), False=@n3.XYZ) {group=--- F5 | RVT, parameterName=* RVT - UseInsideProjection, sortPriority=-10}
  n11 = RuntimeVirtualTextureSampleParameter(World Position=@n10) {
   desc=Feedback enabled
   group=--- F5 | RVT
   materialType=BaseColor_Normal_Specular_YCoCg
   parameterName=* RVT - Color/Normal/Roughness/Specular
   sortPriority=1
   virtualTexture=/Game/Athena/Apollo/Maps/Landscape/VT/RVT_Apollo_Landscape
  }
  n12 = RuntimeVirtualTextureSampleParameter!(UVs, World Position=@n10, Mip Level=Constant()) {
   bEnableFeedback=false
   desc=Use mip 0 but don't request feedback. Should give mip requested by landscape.
   group=--- F5 | RVT
   materialType=BaseColor_Normal_Specular_YCoCg
   mipValueMode=RVTMVM_MipLevel
   parameterName=* RVT - Color/Normal/Roughness/Specular
   sortPriority=1
   virtualTexture=&RVT_Apollo_Landscape
  }
  n13 = StaticSwitchParameter(True=@n11.BaseColor, False=@n12.BaseColor) {
   desc=Can be used to help keep low res VT pages being used when on assets far above the terrain
   group=--- F5 | RVT
   parameterName=* RVT - UseRVTFeedback
   sortPriority=-15
  }
  FunctionOutput(None=@n13) {outputName=RVT Base Color}
  n14 = StaticSwitchParameter(True=@n11.Specular, False=@n12.Specular) {
   desc=Can be used to help keep low res VT pages being used when on assets far above the terrain
   group=--- F5 | RVT
   parameterName=* RVT - UseRVTFeedback
   sortPriority=-15
  }
  FunctionOutput(None=@n14) {outputName=RVT Specular}
  n15 = StaticSwitchParameter(True=@n11.Roughness, False=@n12.Roughness) {
   desc=Can be used to help keep low res VT pages being used when on assets far above the terrain
   group=--- F5 | RVT
   parameterName=* RVT - UseRVTFeedback
   sortPriority=-15
  }
  FunctionOutput(None=@n15) {outputName=RVT Roughness}
  n16 = StaticSwitchParameter(True=@n11.Normal, False=@n12.Normal) {
   desc=Can be used to help keep low res VT pages being used when on assets far above the terrain
   group=--- F5 | RVT
   parameterName=* RVT - UseRVTFeedback
   sortPriority=-15
  }
  FunctionOutput(None=@n16) {outputName=RVT Normals}
  If(A=@n7, "A > B"=@n7, "A == B"=@n7, "A < B"=Constant3Vector() {constant=<rgba 1 0 0 0>}) {constB=1}
  NamedRerouteDeclaration(None=FunctionInput(Preview=StaticBool()) {bUsePreviewValueAsDefault=true, inputName=Use RVT, inputType=FunctionInput_StaticBool, sortPriority=-120}) {name=Use RVT, nodeColor=<rgba 0 0 1 1>}
  n17 = RuntimeVirtualTextureSampleParameter!(UVs, World Position) {
   group=--- F5 | RVT
   materialType=WorldHeight
   parameterName=* RVT - Height
   sortPriority=1
   virtualTexture=/Game/Athena/Apollo/Maps/Landscape/VT/RVT_Apollo_Landscape_Height
  }
  NamedRerouteDeclaration(None=@n17.WorldHeight) {name=RVT Height as is, nodeColor=<rgba 1 1 0 1>}
  n18 = WorldPosition()
  n19 = NamedRerouteUsage()
  NamedRerouteDeclaration(None=Subtract(A=@n18.Z, B=@n19)) {name=RVTHeightWithoutOffset, nodeColor=<rgba 1 1 0 1>}
  n20 = ScalarParameter() {group=--- F5 | RVT, parameterName=* RVT - Height Offset, sortPriority=4}
  NamedRerouteDeclaration(None=Subtract(A=Add(A=@n20, B=@n18.Z), B=@n19)) {name=RVT Height, nodeColor=<rgba 1 1 0 1>}
 ]}
