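# Material function for rock materials that tints base color, roughness and specular through a combined procedural mask, with a static tinting switch.
# The mask layers optional baked AO/curvature/painted channels, world-space noise, Z gradient and two world-space band patterns, then applies invert, bias and intensity.
# If nodes select Min or Max blending and inversion as a cheaper lerp alternative.
# It outputs blended MaterialAttributes and the mask, with a debug masking toggle.
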MaterialFunction /BRCoreMaterials/Materials/MaterialFunctions/MF_Rocks_Tinting
graph {expressions {
  MaterialExpressionAbs_0 {class=MaterialExpressionAbs, inputs [None {from=MaterialExpressionAdd_10, output=""}], properties {}}
  MaterialExpressionAbs_4 {class=MaterialExpressionAbs, inputs [None {from=MaterialExpressionAdd_11, output=""}], properties {}}
  MaterialExpressionAdd_0 {
   class=MaterialExpressionAdd
   inputs [
    A {from=MaterialExpressionReroute_20}
    B {from=MaterialExpressionScalarParameter_0}
   ]
  }
  MaterialExpressionAdd_1 {
   class=MaterialExpressionAdd
   inputs [
    A {from=MaterialExpressionIf_2}
    B {from=MaterialExpressionNamedRerouteUsage_85}
   ]
  }
  MaterialExpressionAdd_10 {
   class=MaterialExpressionAdd
   inputs [
    A {from=MaterialExpressionMultiply_10}
    B {from=""}
   ]
   properties {constB=-1}
  }
  MaterialExpressionAdd_11 {
   class=MaterialExpressionAdd
   inputs [
    A {from=MaterialExpressionMultiply_12}
    B {from=""}
   ]
   properties {constB=-1}
  }
  MaterialExpressionAdd_12 {
   class=MaterialExpressionAdd
   inputs [
    A {from=MaterialExpressionMultiply_13}
    B {from=MaterialExpressionNamedRerouteUsage_112}
   ]
  }
  MaterialExpressionAdd_13 {
   class=MaterialExpressionAdd
   inputs [
    A {from=MaterialExpressionMultiply_11}
    B {from=MaterialExpressionNamedRerouteUsage_110}
   ]
  }
  MaterialExpressionAdd_14 {
   class=MaterialExpressionAdd
   inputs [
    A {from=MaterialExpressionSaturate_6}
    B {from=MaterialExpressionSaturate_0}
   ]
  }
  MaterialExpressionAdd_15 {
   class=MaterialExpressionAdd
   inputs [
    A {from=MaterialExpressionReroute_0}
    B {from=MaterialExpressionSubtract_7}
   ]
  }
  MaterialExpressionAdd_2 {
   class=MaterialExpressionAdd
   inputs [
    A {from=MaterialExpressionReroute_18}
    B {from=MaterialExpressionScalarParameter_4}
   ]
  }
  MaterialExpressionAdd_4 {
   class=MaterialExpressionAdd
   inputs [
    A {from=MaterialExpressionIf_4}
    B {from=MaterialExpressionNamedRerouteUsage_104}
   ]
  }
  MaterialExpressionAdd_46 {
   class=MaterialExpressionAdd
   inputs [
    A {from=MaterialExpressionIf_1}
    B {from=MaterialExpressionNamedRerouteUsage_76}
   ]
  }
  MaterialExpressionComponentMask_0 {class=MaterialExpressionComponentMask, inputs [None {from=MaterialExpressionDivide_3}], properties {b=true}}
  MaterialExpressionComponentMask_1 {class=MaterialExpressionComponentMask, inputs [None {from=MaterialExpressionTransform_0}], properties {b=true}}
  MaterialExpressionConstant3Vector_0 {class=MaterialExpressionConstant3Vector}
  MaterialExpressionConstant_40 {class=MaterialExpressionConstant, properties {r=1}}
  MaterialExpressionConstant_41 {class=MaterialExpressionConstant, properties {r=1}}
  MaterialExpressionConstant_42 {class=MaterialExpressionConstant, properties {r=911}}
  MaterialExpressionConstant_44 {class=MaterialExpressionConstant, properties {r=-0.1}}
  MaterialExpressionConstant_45 {class=MaterialExpressionConstant, properties {r=3}}
  MaterialExpressionConstant_46 {class=MaterialExpressionConstant, properties {r=211}}
  MaterialExpressionConstant_47 {class=MaterialExpressionConstant, properties {r=-0.1}}
  MaterialExpressionConstant_48 {class=MaterialExpressionConstant, properties {r=3}}
  MaterialExpressionConvert_0 {
   class=MaterialExpressionConvert
   inputs [
    None {from=MaterialExpressionNamedRerouteUsage_0}
    None {from=MaterialExpressionNamedRerouteUsage_2}
    None {from=MaterialExpressionNamedRerouteUsage_5}
   ]
   properties {
    nodeName=Make Vector3
    convertInputs [
     {
      type=Scalar
      expressionInput {
       expression=&self:MaterialExpressionNamedRerouteUsage_0
       inputName=None
       mask=0
       maskA=0
       maskB=0
       maskG=0
       maskR=0
       outputIndex=0
      }
     }
     {
      type=Scalar
      expressionInput {
       expression=&self:MaterialExpressionNamedRerouteUsage_2
       inputName=None
       mask=0
       maskA=0
       maskB=0
       maskG=0
       maskR=0
       outputIndex=0
      }
     }
     {
      type=Scalar
      expressionInput {
       expression=&self:MaterialExpressionNamedRerouteUsage_5
       inputName=None
       mask=0
       maskA=0
       maskB=0
       maskG=0
       maskR=0
       outputIndex=0
      }
     }
    ]
    convertOutputs [{type=Vector3}]
   }
  }
  MaterialExpressionDesaturation_0 {
   class=MaterialExpressionDesaturation
   inputs [
    None {from=MaterialExpressionReroute_138, output=""}
    Fraction {from=MaterialExpressionScalarParameter_23, output=""}
   ]
   properties {}
  }
  MaterialExpressionDivide_1 {
   class=MaterialExpressionDivide
   inputs [
    A {from=MaterialExpressionReroute_68}
    B {from=MaterialExpressionNamedRerouteUsage_22}
   ]
   properties {constB=500}
  }
  MaterialExpressionDivide_2 {
   class=MaterialExpressionDivide
   inputs [
    A {from=MaterialExpressionReroute_68}
    B {from=MaterialExpressionNamedRerouteUsage_10}
   ]
   properties {constB=500}
  }
  MaterialExpressionDivide_3 {
   class=MaterialExpressionDivide
   inputs [
    A {from=MaterialExpressionSubtract_2}
    B {from=MaterialExpressionMaterialFunctionCall_10, output=Local Bounds Size}
   ]
  }
  MaterialExpressionDotProduct_2 {
   class=MaterialExpressionDotProduct
   inputs [
    A {from=MaterialExpressionNamedRerouteUsage_83, output=""}
    B {from=MaterialExpressionNamedRerouteUsage_90, output=""}
   ]
   properties {}
  }
  MaterialExpressionDotProduct_3 {
   class=MaterialExpressionDotProduct
   inputs [
    A {from=MaterialExpressionNamedRerouteUsage_88, output=""}
    B {from=MaterialExpressionNamedRerouteUsage_106, output=""}
   ]
   properties {}
  }
  MaterialExpressionFrac_0 {class=MaterialExpressionFrac, inputs [None {from=MaterialExpressionDivide_1, output=""}], properties {}}
  MaterialExpressionFrac_4 {class=MaterialExpressionFrac, inputs [None {from=MaterialExpressionDivide_2, output=""}], properties {}}
  MaterialExpressionFunctionInput_0 {
   class=MaterialExpressionFunctionInput
   inputs [Preview {from=MaterialExpressionSetMaterialAttributes_1, output=""}]
   properties {inputName=Input MaterialAttributes, inputType=FunctionInput_MaterialAttributes, sortPriority=-100}
  }
  MaterialExpressionFunctionInput_1 {
   class=MaterialExpressionFunctionInput
   inputs [Preview {from="", output=""}]
   properties {bUsePreviewValueAsDefault=true, inputName=Tintable Mask, inputType=FunctionInput_Scalar, sortPriority=101}
  }
  MaterialExpressionFunctionInput_12 {
   class=MaterialExpressionFunctionInput
   inputs [Preview {from="", output=""}]
   properties {inputName=Baked Curvature, inputType=FunctionInput_Scalar, sortPriority=41}
  }
  MaterialExpressionFunctionInput_13 {
   class=MaterialExpressionFunctionInput
   inputs [Preview {from="", output=""}]
   properties {inputName=DFAO, inputType=FunctionInput_Scalar, sortPriority=44}
  }
  MaterialExpressionFunctionInput_14 {
   class=MaterialExpressionFunctionInput
   inputs [Preview {from="", output=""}]
   properties {inputName=Z Gradient, inputType=FunctionInput_Scalar, sortPriority=45}
  }
  MaterialExpressionFunctionInput_15 {
   class=MaterialExpressionFunctionInput
   inputs [Preview {from="", output=""}]
   properties {inputName=Baked AO, inputType=FunctionInput_Scalar, sortPriority=42}
  }
  MaterialExpressionFunctionInput_16 {
   class=MaterialExpressionFunctionInput
   inputs [Preview {from="", output=""}]
   properties {inputName=Painted Mask, inputType=FunctionInput_Scalar, sortPriority=43}
  }
  MaterialExpressionFunctionInput_3 {
   class=MaterialExpressionFunctionInput
   inputs [Preview {from="", output=""}]
   properties {bUsePreviewValueAsDefault=true, inputName=Wavy Distortion, inputType=FunctionInput_Scalar, sortPriority=47}
  }
  MaterialExpressionFunctionInput_35 {
   class=MaterialExpressionFunctionInput
   inputs [Preview {from=MaterialExpressionConstant_40, output=""}]
   properties {bUsePreviewValueAsDefault=true, inputName="• WS Bands - Offset", inputType=FunctionInput_Scalar, sortPriority=300}
  }
  MaterialExpressionFunctionInput_36 {
   class=MaterialExpressionFunctionInput
   inputs [Preview {from=MaterialExpressionConstant_41, output=""}]
   properties {bUsePreviewValueAsDefault=true, inputName="• WS Bands - Normal Distortion", inputType=FunctionInput_Scalar, sortPriority=300}
  }
  MaterialExpressionFunctionInput_37 {
   class=MaterialExpressionFunctionInput
   inputs [Preview {from=MaterialExpressionConstant_42, output=""}]
   properties {bUsePreviewValueAsDefault=true, inputName="• WS Bands - 1 - Size", inputType=FunctionInput_Scalar, sortPriority=300}
  }
  MaterialExpressionFunctionInput_39 {
   class=MaterialExpressionFunctionInput
   inputs [Preview {from=MaterialExpressionConstant_44, output=""}]
   properties {bUsePreviewValueAsDefault=true, inputName="• WS Bands - 1 - Bias", inputType=FunctionInput_Scalar, sortPriority=300}
  }
  MaterialExpressionFunctionInput_4 {
   class=MaterialExpressionFunctionInput
   inputs [Preview {from="", output=""}]
   properties {inputName="\\ --- Masks --- /", inputType=FunctionInput_Scalar, sortPriority=100}
  }
  MaterialExpressionFunctionInput_44 {
   class=MaterialExpressionFunctionInput
   inputs [Preview {from=MaterialExpressionConstant_45, output=""}]
   properties {bUsePreviewValueAsDefault=true, inputName="• WS Bands - 1 - Intensity", inputType=FunctionInput_Scalar, sortPriority=300}
  }
  MaterialExpressionFunctionInput_45 {
   class=MaterialExpressionFunctionInput
   inputs [Preview {from=MaterialExpressionConstant_46, output=""}]
   properties {bUsePreviewValueAsDefault=true, inputName="• WS Bands - 2 - Size", inputType=FunctionInput_Scalar, sortPriority=300}
  }
  MaterialExpressionFunctionInput_46 {
   class=MaterialExpressionFunctionInput
   inputs [Preview {from=MaterialExpressionConstant_47, output=""}]
   properties {bUsePreviewValueAsDefault=true, inputName="• WS Bands - 2 - Bias", inputType=FunctionInput_Scalar, sortPriority=300}
  }
  MaterialExpressionFunctionInput_47 {
   class=MaterialExpressionFunctionInput
   inputs [Preview {from=MaterialExpressionConstant_48, output=""}]
   properties {bUsePreviewValueAsDefault=true, inputName="• WS Bands - 2 - Intensity", inputType=FunctionInput_Scalar, sortPriority=300}
  }
  MaterialExpressionFunctionInput_5 {
   class=MaterialExpressionFunctionInput
   inputs [Preview {from="", output=""}]
   properties {inputName="\\ --- Dependencies --- /", inputType=FunctionInput_Scalar, sortPriority=40}
  }
  MaterialExpressionFunctionInput_75 {
   class=MaterialExpressionFunctionInput
   inputs [Preview {from=MaterialExpressionConstant3Vector_0, output=""}]
   properties {bUsePreviewValueAsDefault=true, inputName=WS Noise Masks, sortPriority=46}
  }
  MaterialExpressionFunctionInput_9 {
   class=MaterialExpressionFunctionInput
   inputs [Preview {from=MaterialExpressionStaticBool_0, output=""}]
   properties {bUsePreviewValueAsDefault=true, inputName=Use Tinting, inputType=FunctionInput_StaticBool, sortPriority=-110}
  }
  MaterialExpressionFunctionOutput_0 {
   class=MaterialExpressionFunctionOutput
   inputs [None {from=MaterialExpressionNamedRerouteUsage_1, output=""}]
   properties {outputName=Combined Mask}
  }
  MaterialExpressionFunctionOutput_1 {
   class=MaterialExpressionFunctionOutput
   inputs [None {from=MaterialExpressionStaticSwitch_39, output=""}]
   properties {outputName=Blended MaterialAttributes, sortPriority=-100}
  }
  MaterialExpressionGetMaterialAttributes_0 {
   class=MaterialExpressionGetMaterialAttributes
   inputs [None {from=MaterialExpressionNamedRerouteUsage_115, output=""}]
   properties {attributeGetTypes [#1]}
  }
  MaterialExpressionGetMaterialAttributes_1 {
   class=MaterialExpressionGetMaterialAttributes
   inputs [None {from=MaterialExpressionNamedRerouteUsage_7, output=""}]
   properties {attributeGetTypes [#2, #3]}
  }
  MaterialExpressionGetMaterialAttributes_25 {
   class=MaterialExpressionGetMaterialAttributes
   inputs [None {from=MaterialExpressionNamedRerouteUsage_42, output=""}]
   properties {attributeGetTypes [#4]}
  }
  MaterialExpressionIf_0 {
   class=MaterialExpressionIf
   inputs [
    A {from=MaterialExpressionNamedRerouteUsage_82, output=""}
    B {from="", output=""}
    "A > B" {from=MaterialExpressionMax_3, output=""}
    "A == B" {from="", output=""}
    "A < B" {from=MaterialExpressionMin_2, output=""}
   ]
   properties {constB=0.5, desc="cheaper than lerp without \"==\" condition", equalsThreshold=0.001}
  }
  MaterialExpressionIf_1 {
   class=MaterialExpressionIf
   inputs [
    A {from=MaterialExpressionNamedRerouteUsage_94, output=""}
    B {from="", output=""}
    "A > B" {from=MaterialExpressionOneMinus_0, output=""}
    "A == B" {from="", output=""}
    "A < B" {from=MaterialExpressionDotProduct_2, output=""}
   ]
   properties {constB=0.5, desc="cheaper than lerp without \"==\" condition", equalsThreshold=0.001}
  }
  MaterialExpressionIf_2 {
   class=MaterialExpressionIf
   inputs [
    A {from=MaterialExpressionNamedRerouteUsage_95, output=""}
    B {from="", output=""}
    "A > B" {from=MaterialExpressionOneMinus_3, output=""}
    "A == B" {from="", output=""}
    "A < B" {from=MaterialExpressionDotProduct_3, output=""}
   ]
   properties {constB=0.5, desc="cheaper than lerp without \"==\" condition", equalsThreshold=0.001}
  }
  MaterialExpressionIf_3 {
   class=MaterialExpressionIf
   inputs [
    A {from=MaterialExpressionNamedRerouteUsage_80, output=""}
    B {from="", output=""}
    "A > B" {from=MaterialExpressionMax_25, output=""}
    "A == B" {from="", output=""}
    "A < B" {from=MaterialExpressionMin_24, output=""}
   ]
   properties {constB=0.5, desc="cheaper than lerp without \"==\" condition", equalsThreshold=0.001}
  }
  MaterialExpressionIf_4 {
   class=MaterialExpressionIf
   inputs [
    A {from=MaterialExpressionNamedRerouteUsage_105, output=""}
    B {from="", output=""}
    "A > B" {from=MaterialExpressionOneMinus_2, output=""}
    "A == B" {from="", output=""}
    "A < B" {from=MaterialExpressionReroute_72, output=""}
   ]
   properties {constB=0.5, desc="cheaper than lerp without \"==\" condition", equalsThreshold=0.001}
  }
  MaterialExpressionIf_5 {
   class=MaterialExpressionIf
   inputs [
    A {from=MaterialExpressionNamedRerouteUsage_81, output=""}
    B {from="", output=""}
    "A > B" {from=MaterialExpressionMax_22, output=""}
    "A == B" {from="", output=""}
    "A < B" {from=MaterialExpressionMin_21, output=""}
   ]
   properties {constB=0.5, desc="cheaper than lerp without \"==\" condition", equalsThreshold=0.001}
  }
  MaterialExpressionIf_6 {
   class=MaterialExpressionIf
   inputs [
    A {from=MaterialExpressionNamedRerouteUsage_18, output=""}
    B {from="", output=""}
    "A > B" {from=MaterialExpressionMax_0, output=""}
    "A == B" {from="", output=""}
    "A < B" {from=MaterialExpressionMin_0, output=""}
   ]
   properties {constB=0.5, desc="cheaper than lerp without \"==\" condition", equalsThreshold=0.001}
  }
  MaterialExpressionLinearInterpolate_0 {
   class=MaterialExpressionLinearInterpolate
   inputs [
    A {from=MaterialExpressionReroute_5}
    B {from=MaterialExpressionReroute_7}
    Alpha {from=MaterialExpressionNamedRerouteUsage_3}
   ]
  }
  MaterialExpressionLinearInterpolate_4 {
   class=MaterialExpressionLinearInterpolate
   inputs [
    A {from=MaterialExpressionReroute_20}
    B {from=MaterialExpressionReroute_16}
    Alpha {from=MaterialExpressionNamedRerouteUsage_6}
   ]
  }
  MaterialExpressionLinearInterpolate_6 {
   class=MaterialExpressionLinearInterpolate
   inputs [
    A {from=MaterialExpressionReroute_18}
    B {from=MaterialExpressionReroute_19}
    Alpha {from=MaterialExpressionNamedRerouteUsage_6}
   ]
  }
  MaterialExpressionLinearInterpolate_7 {
   class=MaterialExpressionLinearInterpolate
   inputs [
    A {from=MaterialExpressionMultiply_27}
    B {from=MaterialExpressionVectorParameter_0}
    Alpha {from=MaterialExpressionVectorParameter_0}
   ]
  }
  MaterialExpressionLocalPosition_0 {class=MaterialExpressionLocalPosition, inputs [], properties {}}
  MaterialExpressionMaterialFunctionCall_10 {
   class=MaterialExpressionMaterialFunctionCall
   inputs []
   properties {materialFunction=&ObjectLocalBounds}
  }
  MaterialExpressionMax_0 {
   class=MaterialExpressionMax
   inputs [
    A {from=MaterialExpressionReroute_2}
    B {from=MaterialExpressionReroute_24}
   ]
  }
  MaterialExpressionMax_22 {
   class=MaterialExpressionMax
   inputs [
    A {from=MaterialExpressionReroute_249}
    B {from=MaterialExpressionReroute_250}
   ]
  }
  MaterialExpressionMax_25 {
   class=MaterialExpressionMax
   inputs [
    A {from=MaterialExpressionReroute_261}
    B {from=MaterialExpressionReroute_262}
   ]
  }
  MaterialExpressionMax_3 {
   class=MaterialExpressionMax
   inputs [
    A {from=MaterialExpressionReroute_143}
    B {from=MaterialExpressionReroute_146}
   ]
  }
  MaterialExpressionMin_0 {
   class=MaterialExpressionMin
   inputs [
    A {from=MaterialExpressionReroute_2, output=""}
    B {from=MaterialExpressionReroute_24, output=""}
   ]
   properties {}
  }
  MaterialExpressionMin_2 {
   class=MaterialExpressionMin
   inputs [
    A {from=MaterialExpressionReroute_143, output=""}
    B {from=MaterialExpressionReroute_146, output=""}
   ]
   properties {}
  }
  MaterialExpressionMin_21 {
   class=MaterialExpressionMin
   inputs [
    A {from=MaterialExpressionReroute_249, output=""}
    B {from=MaterialExpressionReroute_250, output=""}
   ]
   properties {}
  }
  MaterialExpressionMin_24 {
   class=MaterialExpressionMin
   inputs [
    A {from=MaterialExpressionReroute_261, output=""}
    B {from=MaterialExpressionReroute_262, output=""}
   ]
   properties {}
  }
  MaterialExpressionMultiply_1 {
   class=MaterialExpressionMultiply
   inputs [
    A {from=MaterialExpressionAdd_4}
    B {from=MaterialExpressionNamedRerouteUsage_103}
   ]
  }
  MaterialExpressionMultiply_10 {
   class=MaterialExpressionMultiply
   inputs [
    A {from=MaterialExpressionFrac_0}
    B {from=""}
   ]
   properties {constB=2}
  }
  MaterialExpressionMultiply_11 {
   class=MaterialExpressionMultiply
   inputs [
    A {from=MaterialExpressionAbs_0}
    B {from=MaterialExpressionAbs_0}
   ]
  }
  MaterialExpressionMultiply_12 {
   class=MaterialExpressionMultiply
   inputs [
    A {from=MaterialExpressionFrac_4}
    B {from=""}
   ]
   properties {constB=2}
  }
  MaterialExpressionMultiply_13 {
   class=MaterialExpressionMultiply
   inputs [
    A {from=MaterialExpressionAbs_4}
    B {from=MaterialExpressionAbs_4}
   ]
  }
  MaterialExpressionMultiply_14 {
   class=MaterialExpressionMultiply
   inputs [
    A {from=MaterialExpressionAdd_13}
    B {from=MaterialExpressionNamedRerouteUsage_111}
   ]
   properties {constB=0.2}
  }
  MaterialExpressionMultiply_15 {
   class=MaterialExpressionMultiply
   inputs [
    A {from=MaterialExpressionOneMinus_5}
    B {from=MaterialExpressionNamedRerouteUsage_116}
   ]
  }
  MaterialExpressionMultiply_27 {
   class=MaterialExpressionMultiply
   inputs [
    A {from=MaterialExpressionDesaturation_0}
    B {from=MaterialExpressionVectorParameter_0}
   ]
  }
  MaterialExpressionMultiply_3 {
   class=MaterialExpressionMultiply
   inputs [
    A {from=MaterialExpressionAdd_1}
    B {from=MaterialExpressionNamedRerouteUsage_360}
   ]
  }
  MaterialExpressionMultiply_74 {
   class=MaterialExpressionMultiply
   inputs [
    A {from=MaterialExpressionAdd_46}
    B {from=MaterialExpressionNamedRerouteUsage_84}
   ]
  }
  MaterialExpressionMultiply_9 {
   class=MaterialExpressionMultiply
   inputs [
    A {from=MaterialExpressionAdd_12}
    B {from=MaterialExpressionNamedRerouteUsage_113}
   ]
   properties {constB=0.2}
  }
  MaterialExpressionNamedRerouteDeclaration_0 {
   class=MaterialExpressionNamedRerouteDeclaration
   inputs [None {from=MaterialExpressionStaticBoolParameter_0}]
   properties {name="Tinting - Use Color, Roughness and Specular", nodeColor=<rgba 0 0 1 1>}
  }
  MaterialExpressionNamedRerouteDeclaration_1 {
   class=MaterialExpressionNamedRerouteDeclaration
   inputs [None {from=MaterialExpressionLinearInterpolate_0}]
   properties {name=Adjusted BaseColor, nodeColor=<rgba 1 0 0.0234 1>}
  }
  MaterialExpressionNamedRerouteDeclaration_10 {
   class=MaterialExpressionNamedRerouteDeclaration
   inputs [None {from=MaterialExpressionAdd_15}]
   properties {name=WS Bands, nodeColor=<rgba 0 1 0.8125 1>}
  }
  MaterialExpressionNamedRerouteDeclaration_11 {
   class=MaterialExpressionNamedRerouteDeclaration
   inputs [None {from=MaterialExpressionFunctionInput_13}]
   properties {name=_DFAO, nodeColor=<rgba 1 0 0.0525 1>}
  }
  MaterialExpressionNamedRerouteDeclaration_12 {
   class=MaterialExpressionNamedRerouteDeclaration
   inputs [None {from=MaterialExpressionFunctionInput_12}]
   properties {name=_Baked Curvature, nodeColor=<rgba 1 0 0.0525 1>}
  }
  MaterialExpressionNamedRerouteDeclaration_13 {
   class=MaterialExpressionNamedRerouteDeclaration
   inputs [None {from=MaterialExpressionFunctionInput_16}]
   properties {name=_Painted Mask, nodeColor=<rgba 1 0 0.0525 1>}
  }
  MaterialExpressionNamedRerouteDeclaration_14 {
   class=MaterialExpressionNamedRerouteDeclaration
   inputs [None {from=MaterialExpressionScalarParameter_2}]
   properties {name="• WS Bands - Blend (1=Max 0=Min)", nodeColor=<rgba 0.5043 1 0 1>}
  }
  MaterialExpressionNamedRerouteDeclaration_16 {
   class=MaterialExpressionNamedRerouteDeclaration
   inputs [None {from=MaterialExpressionScalarParameter_3}]
   properties {name=Tinting - Use Masking, nodeColor=<rgba 0 0 1 1>}
  }
  MaterialExpressionNamedRerouteDeclaration_18 {
   class=MaterialExpressionNamedRerouteDeclaration
   inputs [None {from=MaterialExpressionStaticBoolParameter_1}]
   properties {name=Tinting - DEBUG Masking, nodeColor=<rgba 0 0 1 1>}
  }
  MaterialExpressionNamedRerouteDeclaration_19 {
   class=MaterialExpressionNamedRerouteDeclaration
   inputs [None {from=MaterialExpressionConvert_0}]
   properties {name=Baked AO/Curvature/Painted Texture, nodeColor=<rgba 1 1 1 1>}
  }
  MaterialExpressionNamedRerouteDeclaration_2 {
   class=MaterialExpressionNamedRerouteDeclaration
   inputs [None {from=MaterialExpressionFunctionInput_1}]
   properties {name=_Tintable Mask, nodeColor=<rgba 1 1 1 1>}
  }
  MaterialExpressionNamedRerouteDeclaration_20 {
   class=MaterialExpressionNamedRerouteDeclaration
   inputs [None {from=MaterialExpressionVectorParameter_1, output=RGB}]
   properties {name="• Baked AO/Curvature/Painted - Channel Weights", nodeColor=<rgba 0.5043 1 0 1>}
  }
  MaterialExpressionNamedRerouteDeclaration_21 {
   class=MaterialExpressionNamedRerouteDeclaration
   inputs [None {from=MaterialExpressionComponentMask_0}]
   properties {name=_Z Gradient Custom, nodeColor=<rgba 0 1 0.6484 1>}
  }
  MaterialExpressionNamedRerouteDeclaration_22 {
   class=MaterialExpressionNamedRerouteDeclaration
   inputs [None {from=MaterialExpressionFunctionInput_75}]
   properties {name=_WS Noise Masks, nodeColor=<rgba 1 0 0.0525 1>}
  }
  MaterialExpressionNamedRerouteDeclaration_23 {
   class=MaterialExpressionNamedRerouteDeclaration
   inputs [None {from=MaterialExpressionVectorParameter_2, output=RGB}]
   properties {name="• WS Noise - Channel Weights", nodeColor=<rgba 0.5043 1 0 1>}
  }
  MaterialExpressionNamedRerouteDeclaration_24 {
   class=MaterialExpressionNamedRerouteDeclaration
   inputs [None {from=MaterialExpressionScalarParameter_12}]
   properties {name="• Baked AO/Curvature/Painted - Invert", nodeColor=<rgba 0.5043 1 0 1>}
  }
  MaterialExpressionNamedRerouteDeclaration_25 {
   class=MaterialExpressionNamedRerouteDeclaration
   inputs [None {from=MaterialExpressionScalarParameter_20}]
   properties {name="• WS Noise - Invert", nodeColor=<rgba 0.5043 1 0 1>}
  }
  MaterialExpressionNamedRerouteDeclaration_26 {
   class=MaterialExpressionNamedRerouteDeclaration
   inputs [None {from=MaterialExpressionSaturate_2}]
   properties {name=WS Noise, nodeColor=<rgba 0.5 0 1 1>}
  }
  MaterialExpressionNamedRerouteDeclaration_27 {
   class=MaterialExpressionNamedRerouteDeclaration
   inputs [None {from=MaterialExpressionStaticBoolParameter_38}]
   properties {name=--- Use BAKED AO/CURVATURE/PAINTED MASKING ---, nodeColor=<rgba 0.1979 0.0084 0 1>}
  }
  MaterialExpressionNamedRerouteDeclaration_28 {
   class=MaterialExpressionNamedRerouteDeclaration
   inputs [None {from=MaterialExpressionScalarParameter_8}]
   properties {name="• Baked Curvature/AO/Painted - Intensity", nodeColor=<rgba 0.5043 1 0 1>}
  }
  MaterialExpressionNamedRerouteDeclaration_29 {
   class=MaterialExpressionNamedRerouteDeclaration
   inputs [None {from=MaterialExpressionScalarParameter_9}]
   properties {name="• Baked AO/Curvature/Painted - Bias", nodeColor=<rgba 0.5043 1 0 1>}
  }
  MaterialExpressionNamedRerouteDeclaration_3 {
   class=MaterialExpressionNamedRerouteDeclaration
   inputs [None {from=MaterialExpressionFunctionInput_3}]
   properties {name=_Wavy Distortion, nodeColor=<rgba 1 0 0.0525 1>}
  }
  MaterialExpressionNamedRerouteDeclaration_30 {
   class=MaterialExpressionNamedRerouteDeclaration
   inputs [None {from=MaterialExpressionScalarParameter_18}]
   properties {name="• Baked AO/Curvature/Painted - Blend (1=Max 0=Min)", nodeColor=<rgba 0.5043 1 0 1>}
  }
  MaterialExpressionNamedRerouteDeclaration_31 {
   class=MaterialExpressionNamedRerouteDeclaration
   inputs [None {from=MaterialExpressionStaticBoolParameter_20}]
   properties {name=--- Use WS NOISE MASKING ---, nodeColor=<rgba 0.1979 0.0084 0 1>}
  }
  MaterialExpressionNamedRerouteDeclaration_32 {
   class=MaterialExpressionNamedRerouteDeclaration
   inputs [None {from=MaterialExpressionScalarParameter_11}]
   properties {name="• WS Noise - Intensity", nodeColor=<rgba 0.5043 1 0 1>}
  }
  MaterialExpressionNamedRerouteDeclaration_33 {
   class=MaterialExpressionNamedRerouteDeclaration
   inputs [None {from=MaterialExpressionScalarParameter_10}]
   properties {name="• WS Noise - Bias", nodeColor=<rgba 0.5043 1 0 1>}
  }
  MaterialExpressionNamedRerouteDeclaration_34 {
   class=MaterialExpressionNamedRerouteDeclaration
   inputs [None {from=MaterialExpressionScalarParameter_19}]
   properties {name="• WS Noise - Blend (1=Max 0=Min)", nodeColor=<rgba 0.5043 1 0 1>}
  }
  MaterialExpressionNamedRerouteDeclaration_35 {
   class=MaterialExpressionNamedRerouteDeclaration
   inputs [None {from=MaterialExpressionStaticBoolParameter_18}]
   properties {name=--- Use Z GRADIENT MASKING ---, nodeColor=<rgba 0.1979 0.0084 0 1>}
  }
  MaterialExpressionNamedRerouteDeclaration_37 {
   class=MaterialExpressionNamedRerouteDeclaration
   inputs [None {from=MaterialExpressionScalarParameter_13}]
   properties {name="• Z Gradient - Height - Top", nodeColor=<rgba 0.5043 1 0 1>}
  }
  MaterialExpressionNamedRerouteDeclaration_4 {
   class=MaterialExpressionNamedRerouteDeclaration
   inputs [None {from=MaterialExpressionStaticBoolParameter_2}]
   properties {name=--- Use WS BANDS MASKING ---, nodeColor=<rgba 0.1979 0.0084 0 1>}
  }
  MaterialExpressionNamedRerouteDeclaration_42 {
   class=MaterialExpressionNamedRerouteDeclaration
   inputs [None {from=MaterialExpressionScalarParameter_14}]
   properties {name="• Z Gradient - Height - Bottom", nodeColor=<rgba 0.5043 1 0 1>}
  }
  MaterialExpressionNamedRerouteDeclaration_43 {
   class=MaterialExpressionNamedRerouteDeclaration
   inputs [None {from=MaterialExpressionScalarParameter_17}]
   properties {name="• Z Gradient - Blend (1=Max 0=Min)", nodeColor=<rgba 0.5043 1 0 1>}
  }
  MaterialExpressionNamedRerouteDeclaration_48 {
   class=MaterialExpressionNamedRerouteDeclaration
   inputs [None {from=MaterialExpressionScalarParameter_5}]
   properties {name=* Masking - Invert, nodeColor=<rgba 0.5043 1 0 1>}
  }
  MaterialExpressionNamedRerouteDeclaration_49 {
   class=MaterialExpressionNamedRerouteDeclaration
   inputs [None {from=MaterialExpressionScalarParameter_7}]
   properties {name="• Masking - Bias", nodeColor=<rgba 0.5043 1 0 1>}
  }
  MaterialExpressionNamedRerouteDeclaration_5 {
   class=MaterialExpressionNamedRerouteDeclaration
   inputs [None {from=MaterialExpressionSetMaterialAttributes_7}]
   properties {name=Adjusted Specular and Roughness MaterialAttributes, nodeColor=<rgba 0 0.3437 1 1>}
  }
  MaterialExpressionNamedRerouteDeclaration_50 {
   class=MaterialExpressionNamedRerouteDeclaration
   inputs [None {from=MaterialExpressionScalarParameter_15}]
   properties {name="• Masking - Intensity", nodeColor=<rgba 0.5043 1 0 1>}
  }
  MaterialExpressionNamedRerouteDeclaration_51 {
   class=MaterialExpressionNamedRerouteDeclaration
   inputs [None {from=MaterialExpressionReroute_73}]
   properties {name=Combined Mask, nodeColor=<rgba 0.5 0 1 1>}
  }
  MaterialExpressionNamedRerouteDeclaration_56 {
   class=MaterialExpressionNamedRerouteDeclaration
   inputs [None {from=MaterialExpressionScalarParameter_33}]
   properties {name="• WS Bands - Offset", nodeColor=<rgba 0.5043 1 0 1>}
  }
  MaterialExpressionNamedRerouteDeclaration_57 {
   class=MaterialExpressionNamedRerouteDeclaration
   inputs [None {from=MaterialExpressionScalarParameter_34}]
   properties {name="• WS Bands - Normal Distortion", nodeColor=<rgba 0.5043 1 0 1>}
  }
  MaterialExpressionNamedRerouteDeclaration_58 {
   class=MaterialExpressionNamedRerouteDeclaration
   inputs [None {from=MaterialExpressionScalarParameter_35}]
   properties {name="• WS Bands - 1 - Size", nodeColor=<rgba 0.5043 1 0 1>}
  }
  MaterialExpressionNamedRerouteDeclaration_6 {
   class=MaterialExpressionNamedRerouteDeclaration
   inputs [None {from=MaterialExpressionFunctionInput_9}]
   properties {name=Use Tinting, nodeColor=<rgba 0 0 1 1>}
  }
  MaterialExpressionNamedRerouteDeclaration_60 {
   class=MaterialExpressionNamedRerouteDeclaration
   inputs [None {from=MaterialExpressionScalarParameter_37}]
   properties {name="• WS Bands - 1 - Bias", nodeColor=<rgba 0.5043 1 0 1>}
  }
  MaterialExpressionNamedRerouteDeclaration_61 {
   class=MaterialExpressionNamedRerouteDeclaration
   inputs [None {from=MaterialExpressionScalarParameter_38}]
   properties {name="• WS Bands - 1 - Intensity", nodeColor=<rgba 0.5043 1 0 1>}
  }
  MaterialExpressionNamedRerouteDeclaration_62 {
   class=MaterialExpressionNamedRerouteDeclaration
   inputs [None {from=MaterialExpressionScalarParameter_39}]
   properties {name="• WS Bands - 2 - Size", nodeColor=<rgba 0.5043 1 0 1>}
  }
  MaterialExpressionNamedRerouteDeclaration_63 {
   class=MaterialExpressionNamedRerouteDeclaration
   inputs [None {from=MaterialExpressionScalarParameter_40}]
   properties {name="• Bands - 2 - Bias", nodeColor=<rgba 0.5043 1 0 1>}
  }
  MaterialExpressionNamedRerouteDeclaration_64 {
   class=MaterialExpressionNamedRerouteDeclaration
   inputs [None {from=MaterialExpressionScalarParameter_41}]
   properties {name="• Bands - 2 - Intensity", nodeColor=<rgba 0.5043 1 0 1>}
  }
  MaterialExpressionNamedRerouteDeclaration_7 {
   class=MaterialExpressionNamedRerouteDeclaration
   inputs [None {from=MaterialExpressionFunctionInput_14}]
   properties {name=_Z Gradient, nodeColor=<rgba 1 0 0.0525 1>}
  }
  MaterialExpressionNamedRerouteDeclaration_8 {
   class=MaterialExpressionNamedRerouteDeclaration
   inputs [None {from=MaterialExpressionFunctionInput_15}]
   properties {name=_Baked AO, nodeColor=<rgba 1 0 0.0525 1>}
  }
  MaterialExpressionNamedRerouteDeclaration_87 {
   class=MaterialExpressionNamedRerouteDeclaration
   inputs [None {from=MaterialExpressionSaturate_36}]
   properties {name=Baked AO/Curvature/Painted, nodeColor=<rgba 0.5 0 1 1>}
  }
  MaterialExpressionNamedRerouteDeclaration_88 {
   class=MaterialExpressionNamedRerouteDeclaration
   inputs [None {from=MaterialExpressionSmoothStep_1}]
   properties {name=Z Gradient, nodeColor=<rgba 0.3828 0 1 1>}
  }
  MaterialExpressionNamedRerouteDeclaration_9 {
   class=MaterialExpressionNamedRerouteDeclaration
   inputs [None {from=MaterialExpressionReroute_15}]
   properties {name=_Input MaterialAttributes, nodeColor=<rgba 0.1016 0 1 1>}
  }
  MaterialExpressionNamedRerouteUsage_0 {class=MaterialExpressionNamedRerouteUsage}
  MaterialExpressionNamedRerouteUsage_1 {class=MaterialExpressionNamedRerouteUsage}
  MaterialExpressionNamedRerouteUsage_10 {class=MaterialExpressionNamedRerouteUsage}
  MaterialExpressionNamedRerouteUsage_100 {class=MaterialExpressionNamedRerouteUsage}
  MaterialExpressionNamedRerouteUsage_101 {class=MaterialExpressionNamedRerouteUsage}
  MaterialExpressionNamedRerouteUsage_102 {class=MaterialExpressionNamedRerouteUsage}
  MaterialExpressionNamedRerouteUsage_103 {class=MaterialExpressionNamedRerouteUsage}
  MaterialExpressionNamedRerouteUsage_104 {class=MaterialExpressionNamedRerouteUsage}
  MaterialExpressionNamedRerouteUsage_105 {class=MaterialExpressionNamedRerouteUsage}
  MaterialExpressionNamedRerouteUsage_106 {class=MaterialExpressionNamedRerouteUsage}
  MaterialExpressionNamedRerouteUsage_11 {class=MaterialExpressionNamedRerouteUsage}
  MaterialExpressionNamedRerouteUsage_110 {class=MaterialExpressionNamedRerouteUsage}
  MaterialExpressionNamedRerouteUsage_111 {class=MaterialExpressionNamedRerouteUsage}
  MaterialExpressionNamedRerouteUsage_112 {class=MaterialExpressionNamedRerouteUsage}
  MaterialExpressionNamedRerouteUsage_113 {class=MaterialExpressionNamedRerouteUsage}
  MaterialExpressionNamedRerouteUsage_115 {class=MaterialExpressionNamedRerouteUsage}
  MaterialExpressionNamedRerouteUsage_116 {class=MaterialExpressionNamedRerouteUsage}
  MaterialExpressionNamedRerouteUsage_12 {class=MaterialExpressionNamedRerouteUsage}
  MaterialExpressionNamedRerouteUsage_13 {class=MaterialExpressionNamedRerouteUsage}
  MaterialExpressionNamedRerouteUsage_14 {class=MaterialExpressionNamedRerouteUsage}
  MaterialExpressionNamedRerouteUsage_15 {class=MaterialExpressionNamedRerouteUsage}
  MaterialExpressionNamedRerouteUsage_16 {class=MaterialExpressionNamedRerouteUsage}
  MaterialExpressionNamedRerouteUsage_17 {class=MaterialExpressionNamedRerouteUsage}
  MaterialExpressionNamedRerouteUsage_173 {class=MaterialExpressionNamedRerouteUsage}
  MaterialExpressionNamedRerouteUsage_174 {class=MaterialExpressionNamedRerouteUsage}
  MaterialExpressionNamedRerouteUsage_18 {class=MaterialExpressionNamedRerouteUsage}
  MaterialExpressionNamedRerouteUsage_2 {class=MaterialExpressionNamedRerouteUsage}
  MaterialExpressionNamedRerouteUsage_22 {class=MaterialExpressionNamedRerouteUsage}
  MaterialExpressionNamedRerouteUsage_24 {class=MaterialExpressionNamedRerouteUsage}
  MaterialExpressionNamedRerouteUsage_25 {class=MaterialExpressionNamedRerouteUsage}
  MaterialExpressionNamedRerouteUsage_26 {class=MaterialExpressionNamedRerouteUsage}
  MaterialExpressionNamedRerouteUsage_28 {class=MaterialExpressionNamedRerouteUsage}
  MaterialExpressionNamedRerouteUsage_3 {class=MaterialExpressionNamedRerouteUsage}
  MaterialExpressionNamedRerouteUsage_34 {class=MaterialExpressionNamedRerouteUsage}
  MaterialExpressionNamedRerouteUsage_360 {class=MaterialExpressionNamedRerouteUsage}
  MaterialExpressionNamedRerouteUsage_39 {class=MaterialExpressionNamedRerouteUsage}
  MaterialExpressionNamedRerouteUsage_4 {class=MaterialExpressionNamedRerouteUsage}
  MaterialExpressionNamedRerouteUsage_42 {class=MaterialExpressionNamedRerouteUsage}
  MaterialExpressionNamedRerouteUsage_45 {class=MaterialExpressionNamedRerouteUsage}
  MaterialExpressionNamedRerouteUsage_46 {class=MaterialExpressionNamedRerouteUsage}
  MaterialExpressionNamedRerouteUsage_47 {class=MaterialExpressionNamedRerouteUsage}
  MaterialExpressionNamedRerouteUsage_48 {class=MaterialExpressionNamedRerouteUsage}
  MaterialExpressionNamedRerouteUsage_49 {class=MaterialExpressionNamedRerouteUsage}
  MaterialExpressionNamedRerouteUsage_5 {class=MaterialExpressionNamedRerouteUsage}
  MaterialExpressionNamedRerouteUsage_50 {class=MaterialExpressionNamedRerouteUsage}
  MaterialExpressionNamedRerouteUsage_51 {class=MaterialExpressionNamedRerouteUsage}
  MaterialExpressionNamedRerouteUsage_55 {class=MaterialExpressionNamedRerouteUsage}
  MaterialExpressionNamedRerouteUsage_6 {class=MaterialExpressionNamedRerouteUsage}
  MaterialExpressionNamedRerouteUsage_67 {class=MaterialExpressionNamedRerouteUsage}
  MaterialExpressionNamedRerouteUsage_68 {class=MaterialExpressionNamedRerouteUsage}
  MaterialExpressionNamedRerouteUsage_69 {class=MaterialExpressionNamedRerouteUsage}
  MaterialExpressionNamedRerouteUsage_7 {class=MaterialExpressionNamedRerouteUsage}
  MaterialExpressionNamedRerouteUsage_70 {class=MaterialExpressionNamedRerouteUsage}
  MaterialExpressionNamedRerouteUsage_71 {class=MaterialExpressionNamedRerouteUsage}
  MaterialExpressionNamedRerouteUsage_72 {class=MaterialExpressionNamedRerouteUsage}
  MaterialExpressionNamedRerouteUsage_73 {class=MaterialExpressionNamedRerouteUsage}
  MaterialExpressionNamedRerouteUsage_74 {class=MaterialExpressionNamedRerouteUsage}
  MaterialExpressionNamedRerouteUsage_75 {class=MaterialExpressionNamedRerouteUsage}
  MaterialExpressionNamedRerouteUsage_76 {class=MaterialExpressionNamedRerouteUsage}
  MaterialExpressionNamedRerouteUsage_77 {class=MaterialExpressionNamedRerouteUsage}
  MaterialExpressionNamedRerouteUsage_78 {class=MaterialExpressionNamedRerouteUsage}
  MaterialExpressionNamedRerouteUsage_79 {class=MaterialExpressionNamedRerouteUsage}
  MaterialExpressionNamedRerouteUsage_80 {class=MaterialExpressionNamedRerouteUsage}
  MaterialExpressionNamedRerouteUsage_81 {class=MaterialExpressionNamedRerouteUsage}
  MaterialExpressionNamedRerouteUsage_82 {class=MaterialExpressionNamedRerouteUsage}
  MaterialExpressionNamedRerouteUsage_83 {class=MaterialExpressionNamedRerouteUsage}
  MaterialExpressionNamedRerouteUsage_84 {class=MaterialExpressionNamedRerouteUsage}
  MaterialExpressionNamedRerouteUsage_85 {class=MaterialExpressionNamedRerouteUsage}
  MaterialExpressionNamedRerouteUsage_87 {class=MaterialExpressionNamedRerouteUsage}
  MaterialExpressionNamedRerouteUsage_88 {class=MaterialExpressionNamedRerouteUsage}
  MaterialExpressionNamedRerouteUsage_90 {class=MaterialExpressionNamedRerouteUsage}
  MaterialExpressionNamedRerouteUsage_94 {class=MaterialExpressionNamedRerouteUsage}
  MaterialExpressionNamedRerouteUsage_95 {class=MaterialExpressionNamedRerouteUsage}
  MaterialExpressionNamedRerouteUsage_97 {class=MaterialExpressionNamedRerouteUsage}
  MaterialExpressionNamedRerouteUsage_98 {class=MaterialExpressionNamedRerouteUsage}
  MaterialExpressionNamedRerouteUsage_99 {class=MaterialExpressionNamedRerouteUsage}
  MaterialExpressionOneMinus_0 {class=MaterialExpressionOneMinus, inputs [None {from=MaterialExpressionDotProduct_2}]}
  MaterialExpressionOneMinus_2 {class=MaterialExpressionOneMinus, inputs [None {from=MaterialExpressionReroute_72}]}
  MaterialExpressionOneMinus_3 {class=MaterialExpressionOneMinus, inputs [None {from=MaterialExpressionDotProduct_3}]}
  MaterialExpressionOneMinus_5 {class=MaterialExpressionOneMinus, inputs [None {from=MaterialExpressionComponentMask_1}]}
  MaterialExpressionReroute_0 {class=MaterialExpressionReroute, inputs [None {from=MaterialExpressionAdd_14}]}
  MaterialExpressionReroute_1 {class=MaterialExpressionReroute, inputs [None {from=MaterialExpressionReroute_2}]}
  MaterialExpressionReroute_10 {class=MaterialExpressionReroute, inputs [None {from=MaterialExpressionNamedRerouteUsage_69}]}
  MaterialExpressionReroute_11 {class=MaterialExpressionReroute, inputs [None {from=MaterialExpressionNamedRerouteUsage_39}]}
  MaterialExpressionReroute_12 {class=MaterialExpressionReroute, inputs [None {from=MaterialExpressionNamedRerouteUsage_24}]}
  MaterialExpressionReroute_13 {class=MaterialExpressionReroute, inputs [None {from=MaterialExpressionNamedRerouteUsage_67}]}
  MaterialExpressionReroute_138 {
   class=MaterialExpressionReroute
   inputs [None {from=MaterialExpressionGetMaterialAttributes_25, output=BaseColor}]
   properties {desc=Base Color}
  }
  MaterialExpressionReroute_14 {class=MaterialExpressionReroute, inputs [None {from=MaterialExpressionNamedRerouteUsage_50}]}
  MaterialExpressionReroute_142 {class=MaterialExpressionReroute, inputs [None {from=MaterialExpressionReroute_143}]}
  MaterialExpressionReroute_143 {class=MaterialExpressionReroute, inputs [None {from=MaterialExpressionReroute_64}]}
  MaterialExpressionReroute_146 {class=MaterialExpressionReroute, inputs [None {from=MaterialExpressionNamedRerouteUsage_173}]}
  MaterialExpressionReroute_15 {class=MaterialExpressionReroute, inputs [None {from=MaterialExpressionFunctionInput_0}]}
  MaterialExpressionReroute_16 {class=MaterialExpressionReroute, inputs [None {from=MaterialExpressionAdd_0}]}
  MaterialExpressionReroute_161 {class=MaterialExpressionReroute, inputs [None {from=MaterialExpressionStaticSwitch_7}]}
  MaterialExpressionReroute_17 {class=MaterialExpressionReroute, inputs [None {from=MaterialExpressionNamedRerouteUsage_51}]}
  MaterialExpressionReroute_18 {class=MaterialExpressionReroute, inputs [None {from=MaterialExpressionGetMaterialAttributes_1, output=Roughness}]}
  MaterialExpressionReroute_19 {class=MaterialExpressionReroute, inputs [None {from=MaterialExpressionAdd_2}]}
  MaterialExpressionReroute_2 {class=MaterialExpressionReroute, inputs [None {from=MaterialExpressionReroute_6}]}
  MaterialExpressionReroute_20 {class=MaterialExpressionReroute, inputs [None {from=MaterialExpressionGetMaterialAttributes_1, output=Specular}]}
  MaterialExpressionReroute_21 {class=MaterialExpressionReroute, inputs [None {from=MaterialExpressionGetMaterialAttributes_1, output=MaterialAttributes}]}
  MaterialExpressionReroute_22 {class=MaterialExpressionReroute, inputs [None {from=MaterialExpressionNamedRerouteUsage_45}]}
  MaterialExpressionReroute_229 {class=MaterialExpressionReroute, inputs [None {from=MaterialExpressionReroute_261}]}
  MaterialExpressionReroute_23 {class=MaterialExpressionReroute, inputs [None {from=MaterialExpressionReroute_16}]}
  MaterialExpressionReroute_231 {class=MaterialExpressionReroute, inputs [None {from=MaterialExpressionReroute_249}]}
  MaterialExpressionReroute_24 {class=MaterialExpressionReroute, inputs [None {from=MaterialExpressionNamedRerouteUsage_13}]}
  MaterialExpressionReroute_249 {class=MaterialExpressionReroute, inputs [None {from=MaterialExpressionReroute_25}]}
  MaterialExpressionReroute_25 {class=MaterialExpressionReroute, inputs [None {from=MaterialExpressionReroute_6}]}
  MaterialExpressionReroute_250 {class=MaterialExpressionReroute, inputs [None {from=MaterialExpressionNamedRerouteUsage_174}]}
  MaterialExpressionReroute_261 {class=MaterialExpressionReroute, inputs [None {from=MaterialExpressionReroute_66}]}
  MaterialExpressionReroute_262 {class=MaterialExpressionReroute, inputs [None {from=MaterialExpressionNamedRerouteUsage_87}]}
  MaterialExpressionReroute_3 {class=MaterialExpressionReroute, inputs [None {from=MaterialExpressionReroute_161}]}
  MaterialExpressionReroute_38 {class=MaterialExpressionReroute, inputs [None {from=MaterialExpressionNamedRerouteUsage_48}]}
  MaterialExpressionReroute_4 {class=MaterialExpressionReroute, inputs [None {from=MaterialExpressionNamedRerouteUsage_70}]}
  MaterialExpressionReroute_5 {class=MaterialExpressionReroute, inputs [None {from=MaterialExpressionReroute_138}]}
  MaterialExpressionReroute_52 {class=MaterialExpressionReroute, inputs [None {from=MaterialExpressionReroute_19}]}
  MaterialExpressionReroute_53 {class=MaterialExpressionReroute, inputs [None {from=MaterialExpressionReroute_23}]}
  MaterialExpressionReroute_56 {class=MaterialExpressionReroute, inputs [None {from=MaterialExpressionReroute_52}]}
  MaterialExpressionReroute_57 {class=MaterialExpressionReroute, inputs [None {from=MaterialExpressionLinearInterpolate_4}]}
  MaterialExpressionReroute_58 {class=MaterialExpressionReroute, inputs [None {from=MaterialExpressionLinearInterpolate_6}]}
  MaterialExpressionReroute_59 {class=MaterialExpressionReroute, inputs [None {from=MaterialExpressionNamedRerouteUsage_46}]}
  MaterialExpressionReroute_6 {class=MaterialExpressionReroute, inputs [None {from=MaterialExpressionNamedRerouteUsage_4}]}
  MaterialExpressionReroute_60 {class=MaterialExpressionReroute, inputs [None {from=MaterialExpressionNamedRerouteUsage_47}]}
  MaterialExpressionReroute_61 {class=MaterialExpressionReroute, inputs [None {from=MaterialExpressionNamedRerouteUsage_78}]}
  MaterialExpressionReroute_62 {class=MaterialExpressionReroute, inputs [None {from=MaterialExpressionNamedRerouteUsage_98}]}
  MaterialExpressionReroute_63 {class=MaterialExpressionReroute, inputs [None {from=MaterialExpressionNamedRerouteUsage_99}]}
  MaterialExpressionReroute_64 {class=MaterialExpressionReroute, inputs [None {from=MaterialExpressionReroute_65}]}
  MaterialExpressionReroute_65 {class=MaterialExpressionReroute, inputs [None {from=MaterialExpressionStaticSwitch_1}]}
  MaterialExpressionReroute_66 {class=MaterialExpressionReroute, inputs [None {from=MaterialExpressionStaticSwitch_2}]}
  MaterialExpressionReroute_67 {class=MaterialExpressionReroute, inputs [None {from=MaterialExpressionStaticSwitch_10}]}
  MaterialExpressionReroute_68 {class=MaterialExpressionReroute, inputs [None {from=MaterialExpressionSubtract_6}]}
  MaterialExpressionReroute_69 {class=MaterialExpressionReroute, inputs [None {from=MaterialExpressionNamedRerouteUsage_102}]}
  MaterialExpressionReroute_7 {class=MaterialExpressionReroute, inputs [None {from=MaterialExpressionLinearInterpolate_7}]}
  MaterialExpressionReroute_70 {class=MaterialExpressionReroute, inputs [None {from=MaterialExpressionNamedRerouteUsage_101}]}
  MaterialExpressionReroute_71 {class=MaterialExpressionReroute, inputs [None {from=MaterialExpressionNamedRerouteUsage_100}]}
  MaterialExpressionReroute_72 {class=MaterialExpressionReroute, inputs [None {from=MaterialExpressionReroute_67}]}
  MaterialExpressionReroute_73 {class=MaterialExpressionReroute, inputs [None {from=MaterialExpressionSaturate_5}]}
  MaterialExpressionReroute_8 {class=MaterialExpressionReroute, inputs [None {from=MaterialExpressionNamedRerouteUsage_71}]}
  MaterialExpressionReroute_9 {class=MaterialExpressionReroute, inputs [None {from=MaterialExpressionNamedRerouteUsage_68}]}
  MaterialExpressionSaturate_0 {class=MaterialExpressionSaturate, inputs [None {from=MaterialExpressionMultiply_9}]}
  MaterialExpressionSaturate_2 {class=MaterialExpressionSaturate, inputs [None {from=MaterialExpressionMultiply_3}]}
  MaterialExpressionSaturate_36 {class=MaterialExpressionSaturate, inputs [None {from=MaterialExpressionMultiply_74}]}
  MaterialExpressionSaturate_5 {class=MaterialExpressionSaturate, inputs [None {from=MaterialExpressionMultiply_1}]}
  MaterialExpressionSaturate_6 {class=MaterialExpressionSaturate, inputs [None {from=MaterialExpressionMultiply_14}]}
  MaterialExpressionScalarParameter_0 {
   class=MaterialExpressionScalarParameter
   properties {
    desc="Reduces Specular on Tinting < 0 > Increases Specular on Tinting"
    group=--- F2 | Tinting
    parameterName=* Tinting - Specular - Bias
    sortPriority=36
   }
  }
  MaterialExpressionScalarParameter_1 {
   class=MaterialExpressionScalarParameter
   properties {bUseCustomPrimitiveData=true, parameterName=PD_Tinting_Z_Gradient_Offset, primitiveDataIndex=1}
  }
  MaterialExpressionScalarParameter_10 {
   class=MaterialExpressionScalarParameter
   properties {defaultValue=-0.5, group=--- F2 | Tinting, parameterName=* Tinting - WS Noise - Bias, sortPriority=425}
  }
  MaterialExpressionScalarParameter_11 {
   class=MaterialExpressionScalarParameter
   properties {defaultValue=1, group=--- F2 | Tinting, parameterName=* Tinting - WS Noise - Intensity, sortPriority=426}
  }
  MaterialExpressionScalarParameter_12 {
   class=MaterialExpressionScalarParameter
   properties {group=--- F2 | Tinting, parameterName=* Tinting - Baked AO/Curvature/Painted - Invert, sortPriority=361}
  }
  MaterialExpressionScalarParameter_13 {
   class=MaterialExpressionScalarParameter
   properties {
    defaultValue=1
    desc="1= object top bound\r\n0= object bottom bound"
    group=--- F2 | Tinting
    parameterName=* Tinting - Z Gradient - Height - Top
    sortPriority=266
   }
  }
  MaterialExpressionScalarParameter_14 {
   class=MaterialExpressionScalarParameter
   properties {
    desc="1= object top bound\r\n0= object bottom bound"
    group=--- F2 | Tinting
    parameterName=* Tinting - Z Gradient - Height - Bottom
    sortPriority=265
   }
  }
  MaterialExpressionScalarParameter_15 {
   class=MaterialExpressionScalarParameter
   properties {defaultValue=1, group=--- F2 | Tinting, parameterName=* Tinting - Masking - Intensity, sortPriority=61}
  }
  MaterialExpressionScalarParameter_17 {
   class=MaterialExpressionScalarParameter
   properties {defaultValue=1, group=--- F2 | Tinting, parameterName="* Tinting - Z Gradient - Blend (1=Max 0=Min)", sortPriority=260}
  }
  MaterialExpressionScalarParameter_18 {
   class=MaterialExpressionScalarParameter
   properties {defaultValue=1, group=--- F2 | Tinting, parameterName="* Tinting - Baked AO/Curvature/Painted - Blend (1=Max 0=Min)", sortPriority=370}
  }
  MaterialExpressionScalarParameter_19 {
   class=MaterialExpressionScalarParameter
   properties {defaultValue=1, group=--- F2 | Tinting, parameterName="* Tinting - WS Noise - Blend (1=Max 0=Min)", sortPriority=420}
  }
  MaterialExpressionScalarParameter_2 {
   class=MaterialExpressionScalarParameter
   properties {group=--- F2 | Tinting, parameterName="* Tinting - WS Bands - Blend (1=Max 0=Min)", sortPriority=320}
  }
  MaterialExpressionScalarParameter_20 {class=MaterialExpressionScalarParameter, properties {group=--- F2 | Tinting, parameterName=* Tinting - WS Noise - Invert, sortPriority=411}}
  MaterialExpressionScalarParameter_23 {class=MaterialExpressionScalarParameter, properties {group=--- F2 | Tinting, parameterName=* Tinting - Initial Desaturation}}
  MaterialExpressionScalarParameter_3 {
   class=MaterialExpressionScalarParameter
   properties {defaultValue=1, group=--- F2 | Tinting, parameterName=* Tinting - Use Masking, sortPriority=14}
  }
  MaterialExpressionScalarParameter_32 {class=MaterialExpressionScalarParameter, properties {group=Cliff, parameterName=Color Band Z Offset, sortPriority=501}}
  MaterialExpressionScalarParameter_33 {
   class=MaterialExpressionScalarParameter
   properties {defaultValue=1, group=--- F2 | Tinting, parameterName=* Tinting - WS Bands - Offset, sortPriority=325}
  }
  MaterialExpressionScalarParameter_34 {
   class=MaterialExpressionScalarParameter
   properties {defaultValue=1, group=--- F2 | Tinting, parameterName=* Tinting - WS Bands - Normal Distortion, sortPriority=326}
  }
  MaterialExpressionScalarParameter_35 {
   class=MaterialExpressionScalarParameter
   properties {
    defaultValue=911
    desc="For more variety between layers use values that are not multiple between layers (prime numbers are perfect for this)"
    group=--- F2 | Tinting
    parameterName=* Tinting - WS Bands - 1 - Size
    sortPriority=330
   }
  }
  MaterialExpressionScalarParameter_37 {
   class=MaterialExpressionScalarParameter
   properties {defaultValue=-0.1, group=--- F2 | Tinting, parameterName=* Tinting - WS Bands - 1 - Bias, sortPriority=331}
  }
  MaterialExpressionScalarParameter_38 {
   class=MaterialExpressionScalarParameter
   properties {defaultValue=3, group=--- F2 | Tinting, parameterName=* Tinting - WS Bands - 1 - Intensity, sortPriority=332}
  }
  MaterialExpressionScalarParameter_39 {
   class=MaterialExpressionScalarParameter
   properties {
    defaultValue=211
    desc="For more variety between layers use values that are not multiple between layers (prime numbers are perfect for this)"
    group=--- F2 | Tinting
    parameterName=* Tinting - WS Bands - 2 - Size
    sortPriority=335
   }
  }
  MaterialExpressionScalarParameter_4 {
   class=MaterialExpressionScalarParameter
   properties {
    desc="Reduces Roughness on Tinting(Shiny) < 0 > Increases Roughness on Tinting(Dull)"
    group=--- F2 | Tinting
    parameterName=* Tinting - Roughness - Bias
    sortPriority=35
   }
  }
  MaterialExpressionScalarParameter_40 {
   class=MaterialExpressionScalarParameter
   properties {defaultValue=-0.1, group=--- F2 | Tinting, parameterName=* Tinting - WS Bands - 2 - Bias, sortPriority=336}
  }
  MaterialExpressionScalarParameter_41 {
   class=MaterialExpressionScalarParameter
   properties {defaultValue=3, group=--- F2 | Tinting, parameterName=* Tinting - WS Bands - 2 - Intensity, sortPriority=337}
  }
  MaterialExpressionScalarParameter_5 {class=MaterialExpressionScalarParameter, properties {group=--- F2 | Tinting, parameterName=* Tinting - Masking - Invert, sortPriority=50}}
  MaterialExpressionScalarParameter_7 {class=MaterialExpressionScalarParameter, properties {group=--- F2 | Tinting, parameterName=* Tinting - Masking - Bias, sortPriority=60}}
  MaterialExpressionScalarParameter_8 {
   class=MaterialExpressionScalarParameter
   properties {group=--- F2 | Tinting, parameterName=* Tinting - Baked AO/Curvature/Painted - Intensity, sortPriority=376}
  }
  MaterialExpressionScalarParameter_9 {
   class=MaterialExpressionScalarParameter
   properties {group=--- F2 | Tinting, parameterName=* Tinting - Baked AO/Curvature/Painted - Bias, sortPriority=375}
  }
  MaterialExpressionSetMaterialAttributes_1 {class=MaterialExpressionSetMaterialAttributes, inputs [MaterialAttributes {from="", output=""}]}
  MaterialExpressionSetMaterialAttributes_11 {
   class=MaterialExpressionSetMaterialAttributes
   inputs [
    MaterialAttributes {from=MaterialExpressionNamedRerouteUsage_25, output=""}
    BaseColor {from=MaterialExpressionNamedRerouteUsage_34, output=""}
   ]
   properties {attributeSetTypes [#4]}
  }
  MaterialExpressionSetMaterialAttributes_13 {
   class=MaterialExpressionSetMaterialAttributes
   inputs [
    MaterialAttributes {from=MaterialExpressionNamedRerouteUsage_26, output=""}
    BaseColor {from=MaterialExpressionNamedRerouteUsage_28, output=""}
   ]
   properties {attributeSetTypes [#4]}
  }
  MaterialExpressionSetMaterialAttributes_14 {
   class=MaterialExpressionSetMaterialAttributes
   inputs [
    MaterialAttributes {from=MaterialExpressionReroute_3, output=""}
    BaseColor {from=MaterialExpressionNamedRerouteUsage_97, output=""}
   ]
   properties {attributeSetTypes [#4]}
  }
  MaterialExpressionSetMaterialAttributes_7 {
   class=MaterialExpressionSetMaterialAttributes
   inputs [
    MaterialAttributes {from=MaterialExpressionReroute_21, output=""}
    Specular {from=MaterialExpressionReroute_57, output=""}
    Roughness {from=MaterialExpressionReroute_58, output=""}
   ]
   properties {attributeSetTypes [#2, #3]}
  }
  MaterialExpressionSetMaterialAttributes_8 {
   class=MaterialExpressionSetMaterialAttributes
   inputs [
    MaterialAttributes {from=MaterialExpressionReroute_21, output=""}
    Specular {from=MaterialExpressionReroute_53, output=""}
    Roughness {from=MaterialExpressionReroute_56, output=""}
   ]
   properties {attributeSetTypes [#2, #3]}
  }
  MaterialExpressionSmoothStep_1 {
   class=MaterialExpressionSmoothStep
   inputs [
    Min {from=MaterialExpressionNamedRerouteUsage_77, output=""}
    Max {from=MaterialExpressionNamedRerouteUsage_79, output=""}
    Value {from=MaterialExpressionNamedRerouteUsage_49, output=""}
   ]
   properties {}
  }
  MaterialExpressionStaticBoolParameter_0 {
   class=MaterialExpressionStaticBoolParameter
   inputs []
   properties {group=--- F2 | Tinting, parameterName="* Tinting - Use Color, Roughness and Specular", sortPriority=16}
  }
  MaterialExpressionStaticBoolParameter_1 {
   class=MaterialExpressionStaticBoolParameter
   inputs []
   properties {group=--- F2 | Tinting, parameterName=* Tinting - DEBUG Masking, sortPriority=15}
  }
  MaterialExpressionStaticBoolParameter_18 {
   class=MaterialExpressionStaticBoolParameter
   inputs []
   properties {group=--- F2 | Tinting, parameterName=* Tinting --- Use Z GRADIENT MASKING ---, sortPriority=250}
  }
  MaterialExpressionStaticBoolParameter_2 {
   class=MaterialExpressionStaticBoolParameter
   inputs []
   properties {group=--- F2 | Tinting, parameterName=* Tinting --- Use WS BANDS MASKING ---, sortPriority=300}
  }
  MaterialExpressionStaticBoolParameter_20 {
   class=MaterialExpressionStaticBoolParameter
   inputs []
   properties {group=--- F2 | Tinting, parameterName=* Tinting --- Use WS NOISE MASKING ---, sortPriority=400}
  }
  MaterialExpressionStaticBoolParameter_38 {
   class=MaterialExpressionStaticBoolParameter
   inputs []
   properties {group=--- F2 | Tinting, parameterName=* Tinting --- Use BAKED AO/CURVATURE/PAINTED MASKING ---, sortPriority=350}
  }
  MaterialExpressionStaticBool_0 {class=MaterialExpressionStaticBool}
  MaterialExpressionStaticSwitch_0 {
   class=MaterialExpressionStaticSwitch
   inputs [
    True {from=MaterialExpressionSetMaterialAttributes_14, output=""}
    False {from=MaterialExpressionReroute_3, output=""}
    Value {from=MaterialExpressionNamedRerouteUsage_55, output=""}
   ]
   properties {}
  }
  MaterialExpressionStaticSwitch_1 {
   class=MaterialExpressionStaticSwitch
   inputs [
    True {from=MaterialExpressionIf_5, output=""}
    False {from=MaterialExpressionReroute_231, output=""}
    Value {from=MaterialExpressionNamedRerouteUsage_73, output=""}
   ]
   properties {}
  }
  MaterialExpressionStaticSwitch_10 {
   class=MaterialExpressionStaticSwitch
   inputs [
    True {from=MaterialExpressionIf_3, output=""}
    False {from=MaterialExpressionReroute_229, output=""}
    Value {from=MaterialExpressionNamedRerouteUsage_75, output=""}
   ]
   properties {}
  }
  MaterialExpressionStaticSwitch_2 {
   class=MaterialExpressionStaticSwitch
   inputs [
    True {from=MaterialExpressionIf_0, output=""}
    False {from=MaterialExpressionReroute_142, output=""}
    Value {from=MaterialExpressionNamedRerouteUsage_74, output=""}
   ]
   properties {}
  }
  MaterialExpressionStaticSwitch_3 {
   class=MaterialExpressionStaticSwitch
   inputs [
    True {from=MaterialExpressionIf_6, output=""}
    False {from=MaterialExpressionReroute_1, output=""}
    Value {from=MaterialExpressionNamedRerouteUsage_14, output=""}
   ]
   properties {}
  }
  MaterialExpressionStaticSwitch_39 {
   class=MaterialExpressionStaticSwitch
   inputs [
    True {from=MaterialExpressionStaticSwitch_0, output=""}
    False {from=MaterialExpressionNamedRerouteUsage_16, output=""}
    Value {from=MaterialExpressionNamedRerouteUsage_11, output=""}
   ]
   properties {}
  }
  MaterialExpressionStaticSwitch_7 {
   class=MaterialExpressionStaticSwitch
   inputs [
    True {from=MaterialExpressionSetMaterialAttributes_13, output=""}
    False {from=MaterialExpressionSetMaterialAttributes_11, output=""}
    Value {from=MaterialExpressionNamedRerouteUsage_15, output=""}
   ]
   properties {}
  }
  MaterialExpressionSubtract_0 {
   class=MaterialExpressionSubtract
   inputs [
    A {from=MaterialExpressionLocalPosition_0, output=XYZ}
    B {from=MaterialExpressionMaterialFunctionCall_10, output=Local Bounds Minimum}
   ]
  }
  MaterialExpressionSubtract_2 {
   class=MaterialExpressionSubtract
   inputs [
    A {from=MaterialExpressionSubtract_0}
    B {from=MaterialExpressionScalarParameter_1}
   ]
   properties {desc="Subtract for \r\nValues < 0 Lower Gradient Origin\r\nValues > 0 Raise Gradient Origin"}
  }
  MaterialExpressionSubtract_5 {
   class=MaterialExpressionSubtract
   inputs [
    A {from=MaterialExpressionWorldPosition_1, output=Z}
    B {from=MaterialExpressionScalarParameter_32}
   ]
  }
  MaterialExpressionSubtract_6 {
   class=MaterialExpressionSubtract
   inputs [
    A {from=MaterialExpressionSubtract_5}
    B {from=MaterialExpressionNamedRerouteUsage_12}
   ]
  }
  MaterialExpressionSubtract_7 {
   class=MaterialExpressionSubtract
   inputs [
    A {from=MaterialExpressionReroute_0}
    B {from=MaterialExpressionMultiply_15}
   ]
  }
  MaterialExpressionTransform_0 {class=MaterialExpressionTransform, inputs [None {from=MaterialExpressionGetMaterialAttributes_0, output=Normal}]}
  MaterialExpressionVectorParameter_0 {
   class=MaterialExpressionVectorParameter
   properties {
    group=--- F2 | Tinting
    parameterName=* Tinting - Primary Color
    sortPriority=30
    channelNames {a="0=Multiply 1=Solid Color", b="", g="", r=""}
    defaultValue=<rgba 1 1 1 0>
   }
  }
  MaterialExpressionVectorParameter_1 {
   class=MaterialExpressionVectorParameter
   properties {
    group=--- F2 | Tinting
    parameterName=* Tinting - Baked AO/Curvature/Painted - Channel Weights
    sortPriority=360
    channelNames {a="", b=Blue, g=Green, r=Red}
    defaultValue=<rgba 1 0 0 0>
   }
  }
  MaterialExpressionVectorParameter_2 {
   class=MaterialExpressionVectorParameter
   properties {
    group=--- F2 | Tinting
    parameterName=* Tinting - WS Noise - Channel Weights
    sortPriority=410
    channelNames {a="", b=Blue, g=Green, r=Red}
    defaultValue=<rgba 1 0 0 0>
   }
  }
  MaterialExpressionVectorParameter_3 {
   class=MaterialExpressionVectorParameter
   properties {
    group=--- F2 | Tinting
    parameterName=* Tinting - Secondary Color
    sortPriority=30
    channelNames {a="0=Multiply 1=Solid Color", b="", g="", r=""}
    defaultValue=<rgba 1 1 1 0>
   }
  }
  MaterialExpressionWorldPosition_1 {class=MaterialExpressionWorldPosition, inputs [], properties {}}
 }}
