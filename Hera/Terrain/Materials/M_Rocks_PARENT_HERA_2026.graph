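# Parent material for Hera rocks and cliffs, using material attributes with Nanite, instanced meshes and displacement fade.
# PreFX builds up to four layers plus shared curvature, AO, gradient and noise masks.
# Named reroutes link tinting (with NarrowFlea), biome tint, EchoFrost ice and landscape-layer blending functions.
# PostFX takes RVT samples for terrain blending, plus wetness, displacement and destructible options; static switches toggle features per instance.
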Material /Hera/Terrain/Materials/M_Rocks_PARENT_HERA_2026
graph {
 properties {
  bAutomaticallySetUsageInEditor=false
  bEnableDisplacementFade=true
  bTangentSpaceNormal=false
  bUseMaterialAttributes=true
  bUsedWithInstancedStaticMeshes=true
  bUsedWithNanite=true
  usedShadingModels=Default Lit
  displacementFadeRange {endSizePixels=6, startSizePixels=8}
 }
 nodes [
  MaterialFunctionCall!(
   Use Wetness=NamedRerouteUsage()
   Use RVT=NamedRerouteUsage()
   Use Realtime Displacement=NamedRerouteUsage()
   Is Destructible=NamedRerouteUsage()
   Input MaterialAttributes=NamedRerouteUsage()
   "\\ --- Masks --- /"
   Wetness Mask=OneMinus(None=NamedRerouteUsage())
   "\\ --- Dependencies --- /"
   Baked Curvature=NamedRerouteUsage()
   Baked AO=NamedRerouteUsage()
   Painted Mask=NamedRerouteUsage()
   DFAO=NamedRerouteUsage()
   Z Gradient=NamedRerouteUsage()
   WS Noise Masks=NamedRerouteUsage()
   "\\ --- RVT --- /"
   RVT Base Color=NamedRerouteUsage()
   RVT Specular=NamedRerouteUsage()
   RVT Roughness=NamedRerouteUsage()
   RVT Normal=NamedRerouteUsage()
   RVT Height=NamedRerouteUsage()
  ) {materialFunction=/BRCoreMaterials/Materials/MaterialFunctions/MF_Rocks_PostFX}
  n1 = MaterialFunctionCall!(
   Use Realtime Displacement=NamedRerouteUsage()
   Use L1 | Base=NamedRerouteUsage()
   Use L2 | Detail=NamedRerouteUsage()
   Use L3 | Top=NamedRerouteUsage()
   Use L4 | Extra=NamedRerouteUsage()
   Use Curvature Coloring=NamedRerouteUsage()
   Use AO Coloring=NamedRerouteUsage()
  ) {materialFunction=/BRCoreMaterials/Materials/MaterialFunctions/MF_Rocks_PreFX}
  NamedRerouteDeclaration(None=@n1.Baked Curvature) {name=_Baked Curvature, nodeColor=<rgba 0 0.1797 1 1>}
  NamedRerouteDeclaration(None=@n1.Z Gradient) {name=_Z Gradient, nodeColor=<rgba 0 0.1797 1 1>}
  n2 = NamedRerouteUsage()
  NamedRerouteDeclaration(None=MaterialFunctionCall!(
    Use Tinting=NamedRerouteUsage()
    Input MaterialAttributes=MaterialFunctionCall!(Landscape Grass, Diffuse Tex Object, In=@n2, Spec Tex Object, Normal Tex Object, TextureSize, ProjectionContrast, HeightContrast) {materialFunction=/Hera/Terrain/Cliffs/Materials/CosmicThunder/MaterialFunctions/MF_Rocks_NarrowFlea}.Result
    "\\ --- Dependencies --- /"
    Baked Curvature=NamedRerouteUsage()
    Baked AO=NamedRerouteUsage()
    Painted Mask=NamedRerouteUsage()
    DFAO=NamedRerouteUsage()
    Z Gradient=NamedRerouteUsage()
    WS Noise Masks=NamedRerouteUsage()
    Wavy Distortion=NamedRerouteUsage()
    "\\ --- Masks --- /"
    Tintable Mask=NamedRerouteUsage()
    "• WS Bands - 1 - Size"
    "• WS Bands - Normal Distortion"
    "• WS Bands - 1 - Intensity"
    "• WS Bands - 2 - Size"
    "• WS Bands - 2 - Bias"
    "• WS Bands - 2 - Intensity"
    "• WS Bands - Offset"
    "• WS Bands - 1 - Bias"
   ) {materialFunction=/BRCoreMaterials/Materials/MaterialFunctions/MF_Rocks_Tinting}.Blended MaterialAttributes) {name=Tinting Blended MaterialAttributes, nodeColor=<rgba 1 0 0.6094 1>}
  n3 = MaterialFunctionCall() {materialFunction=/Hera/Terrain/Materials/MaterialFunctions/MF_Hera_Landscape_Layers}
  NamedRerouteDeclaration(None=@n3.TerrainMasks - QuadB) {name=TerrainMasks - QuadB, nodeColor=<rgba 0.5 0 1 1>}
  NamedRerouteDeclaration(None=@n3.TerrainMasks - QuadA) {name=TerrainMasks - QuadA, nodeColor=<rgba 0.5 0 1 1>}
  NamedRerouteDeclaration(None=@n3.TerrainMasks - QuadC) {name=TerrainMasks - QuadC, nodeColor=<rgba 0.5 0 1 1>}
  NamedRerouteDeclaration(None=@n3.TerrainMasks - QuadD) {name=TerrainMasks - QuadD, nodeColor=<rgba 0.5 0 1 1>}
  n4 = MaterialFunctionCall!(
   Use Landscape Layer=NamedRerouteUsage()
   Input MaterialAttributes=NamedRerouteUsage()
   "\\ --- Dependencies --- /"
   Baked Curvature=NamedRerouteUsage()
   Baked AO=NamedRerouteUsage()
   Painted Mask=NamedRerouteUsage()
   DFAO=NamedRerouteUsage()
   Z Gradient=NamedRerouteUsage()
   WS Noise Masks=NamedRerouteUsage()
   "\\ --- Landscape/Biome Masks --- /"
   TerrainMasks_QuadA=NamedRerouteUsage()
   TerrainMasks_QuadB=NamedRerouteUsage()
   TerrainMasks_QuadC=NamedRerouteUsage()
   TerrainMasks_QuadD=NamedRerouteUsage()
   Landscape - Grass=NamedRerouteUsage()
   Landscape - Chaparral Grass=NamedRerouteUsage()
   Landscape - Desert Grass=NamedRerouteUsage()
   Landscape - Desert Grass Noise=NamedRerouteUsage()
   Landscape - Gravel=NamedRerouteUsage()
   Landscape - Mud=NamedRerouteUsage()
   Landscape - Sand=NamedRerouteUsage()
   Landscape - Asphalt=NamedRerouteUsage()
   Landscape - Scree=NamedRerouteUsage()
   Landscape - Snow=NamedRerouteUsage()
  ) {materialFunction=/Hera/Terrain/Materials/MaterialFunctions/MF_Hera_Landscape_Layers_Blending}
  NamedRerouteDeclaration(None=@n4.Blended MaterialAttributes) {name=Landscape Layer Blended MaterialAttributes, nodeColor=<rgba 0 0.9531 1 1>}
  n5 = MaterialFunctionCall!(Use RVT=NamedRerouteUsage()) {materialFunction=/BRCoreMaterials/Materials/MaterialFunctions/MF_Rocks_RVT_Sampling}
  NamedRerouteDeclaration(None=@n5.RVT Height) {name=_RVT Height, nodeColor=<rgba 0.3749 0.6555 1 0.5>}
  NamedRerouteDeclaration(None=@n5.RVT Base Color) {name=_RVT Base Color, nodeColor=<rgba 0.3749 0.6555 1 0.5>}
  NamedRerouteDeclaration(None=@n3.Landscape - Grass) {name=Landscape - Grass, nodeColor=<rgba 0.5723 0.25 1 1>}
  NamedRerouteDeclaration(None=@n3.Landscape - Gravel) {name=Landscape - Gravel, nodeColor=<rgba 0.5723 0.25 1 1>}
  NamedRerouteDeclaration(None=@n3.Landscape - Sand) {name=Landscape - Sand, nodeColor=<rgba 0.5723 0.25 1 1>}
  NamedRerouteDeclaration(None=@n3.Landscape - Asphalt) {name=Landscape - Asphalt, nodeColor=<rgba 0.5723 0.25 1 1>}
  NamedRerouteDeclaration(None=@n3.Landscape - Scree) {name=Landscape - Scree, nodeColor=<rgba 0.5723 0.25 1 1>}
  NamedRerouteDeclaration(None=@n3.Landscape - Snow) {name=Landscape - Snow, nodeColor=<rgba 0.4297 0 1 1>}
  NamedRerouteDeclaration(None=@n1.DFAO) {name=_DFAO, nodeColor=<rgba 0 0.1797 1 1>}
  NamedRerouteDeclaration(None=@n3.Landscape - Mud) {name=Landscape - Mud, nodeColor=<rgba 0.5723 0.25 1 1>}
  NamedRerouteDeclaration(None=@n5.RVT Normals) {name=_RVT Normals, nodeColor=<rgba 0.3749 0.6555 1 0.5>}
  NamedRerouteDeclaration(None=@n4.Landscape Layer Blend Mask) {name=Landscape Layer Blend Mask, nodeColor=<rgba 0 0.9531 1 1>}
  NamedRerouteDeclaration(None=@n3.Landscape - Chaparral Grass) {name=Landscape - Chaparral Grass, nodeColor=<rgba 0.5723 0.25 1 1>}
  NamedRerouteDeclaration(None=@n3.Landscape - Desert Grass) {name=Landscape - Desert Grass, nodeColor=<rgba 0.5723 0.25 1 1>}
  NamedRerouteDeclaration(None=@n3.Landscape - Desert Grass Noise) {name=Landscape - Desert Grass Noise, nodeColor=<rgba 0.5723 0.25 1 1>}
  NamedRerouteDeclaration(None=@n5.RVT Specular) {name=_RVT Specular, nodeColor=<rgba 0.3749 0.6555 1 0.5>}
  NamedRerouteDeclaration(None=@n5.RVT Roughness) {name=_RVT Roughness, nodeColor=<rgba 0.3749 0.6555 1 0.5>}
  n6 = StaticBoolParameter() {
   desc="Enables Realtime Displacement Params\r\nStart by going through enabled Layers and turn on \"Layer contributes to Displacement\" for the wanted Layers\r\n*NOTE* Requires \"Enable Tessellation\" and \"Displacement Magnitude\" set in the material instance to see effect."
   group="-        Rock/Cliff Global Features"
   parameterName=Use Realtime Displacement
   sortPriority=320
  }
  NamedRerouteDeclaration(None=@n6) {name="_Use Realtime Displacment(SB)", nodeColor=<rgba 0 0 0 1>}
  n7 = StaticBoolParameter() {
   desc="Enables the Curvature Coloring\r\nAdds a color to the BaseColor based on the curvature masking and param\r\n*NOTE* Should have \"L1 - Base\" Baked Curvature data for best results"
   group="-        Rock/Cliff Global Features"
   parameterName=Use Curvature Coloring
   sortPriority=20
  }
  NamedRerouteDeclaration(None=@n7) {name=_Use Curvature, nodeColor=<rgba 0 0 0 1>}
  NamedRerouteDeclaration(None=@n1.WS Noise Masks) {name=_WS Noise Masks, nodeColor=<rgba 0 0.1797 1 1>}
  n8 = StaticBoolParameter() {
   desc="Enables the AO Coloring\r\nUses a Linear(Replace) or Multiply blend of a color with the BaseColor based on the AO masking and param\r\n*NOTE* Should have \"L1 - Base\" Baked AO data and/or using DFAO for best results"
   group="-        Rock/Cliff Global Features"
   parameterName=Use AO Coloring
   sortPriority=25
  }
  NamedRerouteDeclaration(None=@n8) {name=_Use Dirt, nodeColor=<rgba 0 0 0 1>}
  n9 = StaticBoolParameter() {
   desc="Enables the Tinting feature\r\nUses a Linear(Replace) or Multiply blend of a color with the BaseColor, also can support adjusting specular and roughness\r\n*NOTE* Uses each enabled Layer's \"Is Tintable\" value as a starting mask. Other masking options can be combined"
   group="-        Rock/Cliff Global Features"
   parameterName=Use Tinting
   sortPriority=50
  }
  NamedRerouteDeclaration(None=@n9) {name=_Use Tinting, nodeColor=<rgba 0 0 0 1>}
  n10 = StaticBoolParameter() {
   desc="Enables the Biome Tint\r\nUses the Biome Tint values set in the landscape, use to match biome tinting like on cliffs\r\n*NOTE* Uses each enabled Layer's \"Is Tintable\" value as a mask"
   group="-        Rock/Cliff Global Features"
   parameterName=Use Biome Tint on Tintable Layers
   sortPriority=110
  }
  NamedRerouteDeclaration(None=@n10) {name=_Use Biome Tint on Tintable Layers, nodeColor=<rgba 0 0 0 1>}
  n11 = StaticBoolParameter() {
   desc="Enables RVT(Runtime Virtual Texture) Feature\r\nUsing RVT allows for blending an assets material with the landscape material where they meet\r\nThere are params for adjusting the distance RVT blends on the Sides and Top of an asset, also has params specifically for RVT Normals"
   group="-        Rock/Cliff Global Features"
   parameterName=Use RVT
   sortPriority=230
  }
  NamedRerouteDeclaration(None=@n11) {name=_Use RVT, nodeColor=<rgba 0 0 0 1>}
  n12 = StaticBoolParameter() {
   desc="Enables the Wetness Feature\r\nUses a combination of Specular and Roughness to infer the porosity of the combined layers. Highly rough surfaces have higher porosity(more micro surface holes for water to absorb into the surface and darken/saturate the color as well as smooth out the roughness)\r\nWaterline is a masking option in the Wetness params"
   group="-        Rock/Cliff Global Features"
   parameterName=Use Wetness
   sortPriority=310
  }
  NamedRerouteDeclaration(None=@n12) {name=_Use Wetness, nodeColor=<rgba 0 0 0 1>}
  n13 = StaticBoolParameter() {
   desc="Only enable if this asset can be destroyed\r\n"
   group="-        Rock/Cliff Global Features"
   parameterName=Is Destructible
   sortPriority=-5
  }
  NamedRerouteDeclaration(None=@n13) {name=_Is Destructible, nodeColor=<rgba 0 0 0 1>}
  n14 = StaticBoolParameter() {
   desc="Enables the Landscape Layer\r\nAllows selection of a specific Landscape layer that can then be revealed with masking params"
   group="-        Rock/Cliff Global Features"
   parameterName=Use Landscape Layer
   sortPriority=230
  }
  NamedRerouteDeclaration(None=@n14) {name=_Use Landscape Layer, nodeColor=<rgba 0 0 0 1>}
  n15 = StaticBoolParameter() {
   desc="Enables the fourth layer\r\nThis layer is commonly used for an additional layer on top of the base and detail (ie Moss layer)\r\n*NOTE* shouldn't be used if also using RVT and Landscape"
   group="-        Rock/Cliff Global Features"
   parameterName=Use L4 | Extra
   sortPriority=16
  }
  NamedRerouteDeclaration(None=@n15) {name=_Use L4 - Extra, nodeColor=<rgba 0 0 0 1>}
  NamedRerouteDeclaration(None=@n1.Wavy Distortion) {name=_Wavy Distortion, nodeColor=<rgba 0 0.1797 1 1>}
  NamedRerouteDeclaration(None=StaticBoolParameter() {group="-        Rock/Cliff Global Features", parameterName=Use EchoFrost Ice, sortPriority=220}) {name=_Use EchoFrost Ice, nodeColor=<rgba 0 0 0 1>}
  NamedRerouteDeclaration(None=@n1.Painted Mask) {name=_Painted Mask, nodeColor=<rgba 0 0.1797 1 1>}
  NamedRerouteDeclaration(None=MaterialFunctionCall!(
    Use EchoFrost Ice Layer=NamedRerouteUsage()
    Input MaterialAttributes=NamedRerouteUsage()
    "\\ --- Masks --- /"
    EchoFrost Mask
    Tintable Mask
    "\\ --- Dependencies --- /"
    Baked Curvature=NamedRerouteUsage()
    Baked AO=NamedRerouteUsage()
    Painted Mask=NamedRerouteUsage()
    DFAO=NamedRerouteUsage()
    Z Gradient=NamedRerouteUsage()
    WS Noise Masks=NamedRerouteUsage()
   ) {materialFunction=/Hera/Sets/EchoFrost/Materials/MaterialFunctions/MF_EchoFrost_Ice}.MaterialAttributes) {name=EchoFrost Ice Blended MaterialAttributes, nodeColor=<rgba 0.2969 0.7264 1 1>}
  NamedRerouteDeclaration(None=@n1.Baked AO) {name=_Baked AO, nodeColor=<rgba 0 0.1797 1 1>}
  NamedRerouteDeclaration(None=@n1.PreFX Tintable Mask) {name=PreFX Tintable Mask, nodeColor=<rgba 1 0.0082 0.0025 1>}
  NamedRerouteDeclaration(None=@n1.Rocks PreFX MaterialAttributes) {name=PreFX MaterialAttributes, nodeColor=<rgba 1 0.0082 0.0025 1>}
  n16 = StaticBoolParameter() {
   defaultValue=true
   desc="Enables the first layer\r\nThis layer is primarily meant for baked data (ie Baked Curvature, AO and Painted Msk or Baked Cliff textures)\r\n*NOTE* Does not support \"Fuzzy\" or \"Detail\" layer functionality"
   group="-        Rock/Cliff Global Features"
   parameterName=Use L1 | Base
   sortPriority=5
  }
  NamedRerouteDeclaration(None=@n16) {name=_Use L1 - Base, nodeColor=<rgba 0 0 0 1>}
  n17 = StaticBoolParameter() {
   desc="Enables the second layer\r\nThis layer is commonly used to add higher frequency detail to the object, often with WS tiling"
   group="-        Rock/Cliff Global Features"
   parameterName=Use L2 | Detail
   sortPriority=10
  }
  NamedRerouteDeclaration(None=@n17) {name=_Use L2 - Detail, nodeColor=<rgba 0 0 0 1>}
  n18 = StaticBoolParameter() {
   desc="Enables the third layer\r\nThis layer is commonly used for an additional layer on top of the base and detail (ie Moss layer)"
   group="-        Rock/Cliff Global Features"
   parameterName=Use L3 | Top
   sortPriority=15
  }
  NamedRerouteDeclaration(None=@n18) {name=_Use L3 - Top, nodeColor=<rgba 0 0 0 1>}
  NamedRerouteDeclaration(None=MaterialFunctionCall!(
    Material=NamedRerouteUsage()
    TerrainMasks_QuadA=NamedRerouteUsage()
    TerrainMasks_QuadB=NamedRerouteUsage()
    TerrainMasks_QuadC=NamedRerouteUsage()
    TerrainMasks_QuadD=NamedRerouteUsage()
    MossMask
   ) {materialFunction=/Hera/Environment/Terrain/Landscape/Materials/Layers/BiomeColor/MF_Hera_Cliff_BiomeColor}.Result) {name=Biome Color MaterialAttributes, nodeColor=<rgba 1 0 0.2812 1>}
  NamedRerouteDeclaration(None=MaterialFunctionCall!(
    Use Biome Tint on Tintable Layers=NamedRerouteUsage()
    Input MaterialAttributes=NamedRerouteUsage()
    Biome Color MaterialAttributes=NamedRerouteUsage()
    "\\ --- Masks --- /"
    Tintable Mask=NamedRerouteUsage()
   ) {materialFunction=/BRCoreMaterials/Materials/MaterialFunctions/MF_Rocks_Biome_Tint}.Blended MaterialAttributes) {name=Biome Tint Blended MaterialAttributes, nodeColor=<rgba 0.9922 0 1 1>}
  NamedRerouteUsage()
  NamedRerouteUsage()
  NamedRerouteUsage()
  NamedRerouteUsage()
 ]
}
